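# Material function that drives flipbook-style eyes and brows on the PuzzleShed body by showing only the geometry variants matching per-side index parameters.
# Pre-skinned local X picks left or right indices, which curve atlases remap after eye index 3 becomes 0.
# Sphere masks compare the results to vertex color green for eyes and red for brows.
# Unmatched non-white vertices collapse toward the object position through an offset added to incoming World Position Offset.
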MaterialFunction /BRCosmetics/Characters/Player/Male/Large/Bodies/M_LRG_PuzzleShed/Materials/MF_Agriculture_IndexEyes
graph {nodes [
  n1 = GetMaterialAttributes(None=FunctionInput() {inputName=Material, inputType=FunctionInput_MaterialAttributes}) {attributeGetTypes [#1]}
  n2 = VertexColor()
  n3 = DotProduct(A=Constant3Vector() {constant=<rgba 1 1 1 0>}, B=@n2)
  n4 = Constant() {r=3}
  n5 = Constant()
  n6 = Constant() {r=1}
  n7 = ScalarParameter() {group=Flipbook Eyes, parameterName=flipbook_face_L_eye_index}
  n8 = LinearInterpolate(A=@n7, Alpha=SphereMask(A=@n7, B=Constant() {r=3}) {attenuationRadius=0.02}) {constB=0}
  n9 = ScalarParameter() {group=Flipbook Eyes, parameterName=flipbook_face_R_eye_index}
  n10 = LinearInterpolate(A=@n9, Alpha=SphereMask(A=@n9, B=Constant() {r=3}) {attenuationRadius=0.02}) {constB=0}
  n11 = NamedRerouteUsage()
  n12 = CurveAtlasRowParameter(CurveTime=Divide(A=LinearInterpolate(A=@n8, B=@n10, Alpha=@n11)) {constB=10}) {
   atlas=/Game/Characters/Player/Common/MaterialFunctions/VertexColorFaces/CA_FlipbookV2_Remapping
   curve=/Game/Characters/Player/Common/MaterialFunctions/VertexColorFaces/C_EyesAndMouthDefault
   group=Flipbook Eyes
   parameterName=EyeIndexRemapCurve
  }
  n13 = SphereMask(A=@n12.R, B=VertexColor().G) {attenuationRadius=0.02}
  n14 = ScalarParameter() {group=Flipbook Eyes, parameterName=Flipbook_face_L_brow_index}
  n15 = ScalarParameter() {group=Flipbook Eyes, parameterName=Flipbook_face_R_brow_index}
  n16 = NamedRerouteUsage()
  n17 = CurveAtlasRowParameter(CurveTime=Divide(A=LinearInterpolate(A=@n14, B=@n15, Alpha=@n16)) {constB=10}) {
   atlas=&CA_FlipbookV2_Remapping
   curve=&C_EyesAndMouthDefault
   group=Flipbook Eyes
   parameterName=BrowIndexRemapCurve
  }
  n18 = SphereMask(A=@n17.R, B=VertexColor().R) {attenuationRadius=0.02}
  n19 = Round(None=Subtract(A=If(A=@n3, B=@n4, "A > B"=@n5, "A == B"=@n5, "A < B"=@n6), B=Add(A=@n13, B=@n18)))
  n20 = Add(
   A=NamedRerouteDeclaration(None=@n1.WorldPositionOffset) {name=WPO_IN, nodeColor=<rgba 0.5 1 0 1>}
   B=Multiply(A=NamedRerouteUsage(), B=Saturate(None=@n19))
  )
  FunctionOutput(None=@n20)
  Min(A=Min(A=@n2.R, B=@n2.R), B=@n2.B)
  n21 = Add(A=WorldPosition().XYZ, B=NamedRerouteUsage())
  n22 = ObjectPositionWS()
  NamedRerouteDeclaration(None=Multiply(A=Subtract(A=@n21, B=@n22)) {constB=-1}) {name=Vertex Color Hide, nodeColor=<rgba 0.6875 0 1 1>}
  n23 = LocalPosition() {localOrigin=InstancePreSkinning}
  n24 = VertexInterpolator(VS=@n23.XYZ)
  n25 = Round(None=MaterialFunctionCall(Float3=ShaderStageSwitch(PixelShader=@n24.PS, VertexShader=@n23.XYZ)) {materialFunction=&BreakOutFloat3Components}.R)
  NamedRerouteDeclaration(None=OneMinus(None=Saturate(None=@n25))) {name=Left0/Right1, nodeColor=<rgba 1 0 0.1172 1>}
  SetMaterialAttributes!(MaterialAttributes=@n1.MaterialAttributes, WorldPositionOffset=@n20) {attributeSetTypes [#1]}
 ]}
