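# Material function that draws animated cartoon faces for Agriculture cosmetics, outputting color, alpha, WPO and DFL mask.
# Eye, brow and mouth flipbooks are selected by pose-index parameters through SubUV math, with MF_DFLine shaping their alpha.
# Static switches select TwinkleLoop head or backpack and ChiveFlake backpack variants.
# TwinkleLoop head builds pupils and eyelids procedurally from look and blink poses; ChiveFlake backpack hides vertex-colored expression geometry through WPO.
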MaterialFunction /BRCosmetics/Characters/Player/Male/Large/Bodies/M_LRG_PuzzleShed/Materials/MF_Agriculture_FlipbookFace
graph {nodes [
  n1 = FunctionInput() {bUsePreviewValueAsDefault=true, inputName=WPOIn, sortPriority=1}
  n2 = NamedRerouteUsage()
  n3 = NamedRerouteUsage()
  n4 = NamedRerouteUsage()
  n5 = Constant()
  n6 = StaticSwitchParameter(
   True=MaterialFunctionCall!(Material=SetMaterialAttributes!(MaterialAttributes, WorldPositionOffset=@n1) {attributeSetTypes [#1]}) {materialFunction=&/MF_Agriculture_IndexEyes}.Result
   False=StaticSwitchParameter(
    True=LinearInterpolate(B=@n2, Alpha=@n3)
    False=StaticSwitchParameter(True=@n4, False=@n5) {group=Agriculture Flipbook Face, parameterName=UseChiveFlakeBackpack, sortPriority=0}
   ) {group=Agriculture Flipbook Face, parameterName=UseTwinkleLoopHead, sortPriority=0}
  ) {group=Agriculture Flipbook Face, parameterName=UseTwinkleLoopFaceAcc, sortPriority=0}
  FunctionOutput(None=@n6) {outputName=WPO, sortPriority=2}
  n7 = FunctionInput() {bUsePreviewValueAsDefault=true, inputName=DFL_Mask, inputType=FunctionInput_Scalar, sortPriority=3}
  n8 = NamedRerouteDeclaration(None=@n7) {name=DFL_Mask_IN, nodeColor=<rgba 1 0 0.6094 1>}
  n9 = Add(A=NamedRerouteUsage(), B=NamedRerouteUsage())
  FunctionOutput(None=StaticSwitchParameter(True=LinearInterpolate(A=@n8, Alpha=@n9) {constB=0}, False=@n7) {group=Agriculture Flipbook Face, parameterName=UseTwinkleLoopHead, sortPriority=0}) {outputName=DFL_Mask, sortPriority=3}
  n10 = NamedRerouteUsage()
  n11 = StaticSwitchParameter(
   True=NamedRerouteUsage()
   False=StaticSwitchParameter(True=NamedRerouteUsage(), False=NamedRerouteUsage()) {group=Agriculture Flipbook Face, parameterName=UseChiveFlakeBackpack, sortPriority=0}
  ) {group=Agriculture Flipbook Face, parameterName=UseTwinkleLoopBackpack, sortPriority=0}
  n12 = FunctionInput() {inputName=ColorIn}
  FunctionOutput(None=StaticSwitchParameter(
    True=StaticSwitchParameter(True=@n10, False=@n11) {group=Agriculture Flipbook Face, parameterName=UseTwinkleLoopHead, sortPriority=0}
    False=NamedRerouteDeclaration(None=@n12) {name=Color_IN, nodeColor=<rgba 0 0.0156 1 1>}
   ) {defaultValue=true, group=Agriculture Flipbook Face, parameterName=UseFlipbookFace, sortPriority=-1}) {outputName=Color}
  n13 = StaticSwitchParameter(
   True=NamedRerouteUsage()
   False=StaticSwitchParameter(True=NamedRerouteUsage(), False=NamedRerouteUsage()) {group=Agriculture Flipbook Face, parameterName=UseChiveFlakeBackpack, sortPriority=0}
  ) {group=Agriculture Flipbook Face, parameterName=UseTwinkleLoopBackpack, sortPriority=0}
  FunctionOutput(None=@n13) {outputName=Alpha, sortPriority=1}
  n14 = NamedRerouteUsage()
  n15 = Constant() {r=1}
  n16 = SphereMask(A=@n14, B=@n15) {attenuationRadius=0.02}
  n17 = Constant() {r=3}
  Max(A=@n16, B=SphereMask(A=@n14, B=@n17) {attenuationRadius=1})
  n18 = NamedRerouteUsage()
  n19 = SphereMask(A=@n18, B=@n15) {attenuationRadius=0.02}
  Max(A=@n19, B=SphereMask(A=@n18, B=@n17) {attenuationRadius=1})
  n20 = StaticSwitchParameter(True=TextureCoordinate() {coordinateIndex=1}, False=TextureCoordinate()) {defaultValue=true, group=Agriculture Flipbook Face, parameterName=UseUV2}
  n21 = ComponentMask(None=VectorParameter() {group=Agriculture Flipbook Face, parameterName=UVGloabelScale, defaultValue=<rgba 1 1 0 1>}.RGB) {g=true, r=true}
  NamedRerouteDeclaration(None=Frac(None=MaterialFunctionCall(UVs=@n20, Texture Scale=@n21) {materialFunction=&ScaleUVsByCenter}.UVs)) {name=UVs, nodeColor=<rgba 1 0.2578 0 1>}
  n22 = Add(A=NamedRerouteUsage(), B=NamedRerouteUsage())
  n23 = NamedRerouteUsage()
  NamedRerouteDeclaration(None=Saturate(None=Add(A=@n22, B=@n23))) {name=Alpha_OUT, nodeColor=<rgba 1 0.3047 0 1>}
  n24 = TextureObjectParameter() {
   group=TwinkleLoopBackpack Face
   parameterName=TLB_FlipbookTex
   samplerType=SAMPLERTYPE_Masks
   texture=/BRCosmetics/Accessories/FORT_Backpacks/Backpack_M_MED_TwinkleLoop/Textures/T_TwinkleLoop_Backpack_Flipbook
  }
  n25 = ScalarParameter() {group=TwinkleLoopBackpack Face, parameterName=TLB_Eyes_UVShiftX, sliderMax=1, sliderMin=-1, sortPriority=1}
  n26 = ScalarParameter() {group=TwinkleLoopBackpack Face, parameterName=TLB_Eyes_UVShiftY, sliderMax=1, sliderMin=-1, sortPriority=1}
  n27 = Constant() {r=2}
  n28 = Clamp(None=Floor(None=ScalarParameter() {defaultValue=1, group=TwinkleLoopBackpack Face, parameterName=TLB_Expression_Index, sortPriority=0})) {maxDefault=2}
  n29 = MaterialFunctionCall!(Texture=@n24, UVs=Add(A=NamedRerouteUsage(), B=AppendVector(A=@n25, B=@n26)), SubImages=@n27, Frame=@n28) {materialFunction=/Engine/Functions/Engine_MaterialFunctions03/Texturing/SubUV_Function}
  n30 = Add(
   A=NamedRerouteUsage()
   B=AppendVector(
    A=ScalarParameter() {group=TwinkleLoopBackpack Face, parameterName=TLB_Mouth_UVShiftX, sliderMax=1, sliderMin=-1, sortPriority=1}
    B=ScalarParameter() {group=TwinkleLoopBackpack Face, parameterName=TLB_Mouth_UVShiftY, sliderMax=1, sliderMin=-1, sortPriority=1}
   )
  )
  n31 = Add(
   A=MaterialFunctionCall(Float3=@n29.RGB) {materialFunction=&BreakOutFloat3Components}.R
   B=MaterialFunctionCall(Float3=MaterialFunctionCall(Texture=@n24, UVs=@n30, SubImages=@n27, Frame=@n28) {materialFunction=&SubUV_Function}.RGB) {materialFunction=&BreakOutFloat3Components}.G
  )
  NamedRerouteDeclaration(None=MaterialFunctionCall!(DF=@n31) {materialFunction=&/MF_DFLine}.Result) {name=TLB_Alpha_OUT, nodeColor=<rgba 1 0 0.5391 1>}
  n32 = Constant() {r=0.5}
  MaterialExpressionAdd_0 = Add(
   A=Multiply(A=NamedRerouteUsage(), B=ScalarParameter() {defaultValue=2, group=Agriculture Flipbook Face, parameterName=Mouth_Scale, sortPriority=7})
   B=AppendVector(
    A=ScalarParameter() {group=Agriculture Flipbook Face, parameterName=Mouth_UVShiftX, sortPriority=7}
    B=ScalarParameter() {group=Agriculture Flipbook Face, parameterName=Mouth_UVShiftY, sortPriority=7}
   )
  )
  n33 = MaterialFunctionCall(Float2=@MaterialExpressionAdd_0) {materialFunction=&BreakOutFloat2Components}
  n34 = SphereMask(A=@n32, B=@n33.R) {attenuationRadius=0.5}
  n35 = SphereMask(A=@n33.G, B=@n32) {attenuationRadius=0.5}
  MaterialExpressionScalarParameter_2 = ScalarParameter() {defaultValue=4, group=Agriculture Flipbook Face, parameterName=Mouth_SubImages, sortPriority=7}
  MaterialExpressionNamedRerouteUsage_42 = NamedRerouteUsage()
  n36 = Custom!(uv=@MaterialExpressionAdd_0, size=@MaterialExpressionScalarParameter_2, frame=@MaterialExpressionNamedRerouteUsage_42) {
   code="float size2 = size-0.00001f;\r\nreturn uv/size2 + float2(floor(fmod(frame,size2)), floor(frame/size2))/size2; "
   description=SubUV
   outputType=CMOT_Float2
   inputs [
    {
     inputName=uv
     input {expression=&self:MaterialExpressionAdd_0}
    }
    {
     inputName=size
     input {expression=&self:MaterialExpressionScalarParameter_2}
    }
    {
     inputName=frame
     input {expression=&self:MaterialExpressionNamedRerouteUsage_42}
    }
   ]
  }
  n37 = TextureObjectParameter() {
   group=Agriculture Flipbook Face
   parameterName=Mouth_Flipbook
   texture=/BRCosmetics/Characters/Player/Female/Medium/Bodies/F_MED_ChiveFlake/Textures/T_ChiveFlake_Head_Flipbook_Mouth
  }
  n38 = TextureSample!(UVs=@n36, Tex=@n37, Bias, Apply View MipBias) {
   mipValueMode=TMVM_MipBias
   samplerSource=SSM_Clamp_WorldGroupSettings
   texture=/BRCosmetics/Characters/Player/Female/Medium/Bodies/F_MED_ChiveFlake/Textures/T_ChiveFlake_Head_Flipbook_Eyes
  }
  NamedRerouteDeclaration(None=Multiply(
    A=Multiply(A=@n34, B=@n35)
    B=MaterialFunctionCall(DF=@n38.A) {materialFunction=&MF_DFLine}.Result
   )) {name=Alpha_Mouth, nodeColor=<rgba 1 0 0.1875 1>}
  n39 = ScalarParameter() {defaultValue=1000, group=01_InvertedHull_Outlines, parameterName=Fade_Distance, sliderMax=1500, sliderMin=500, sortPriority=5}
  NamedRerouteDeclaration(None=MaterialFunctionCall!(
    Fade Offset=ScalarParameter() {defaultValue=50, group=01_InvertedHull_Outlines, parameterName=Fade_Offset, sliderMax=100, sortPriority=5}
    Fade Distance=@n39
   ) {materialFunction=/Game/Global/Materials/Character_Parent_2023/Anime_Parent_2023/MaterialFunctions/MFC_DistanceFade_FOV_compensation}.Result) {name=DistanceFading, nodeColor=<rgba 1 0.3047 0 1>}
  NamedRerouteDeclaration(None=@n38.RGB) {name=Color_Mouth, nodeColor=<rgba 0 0.5078 1 1>}
  n40 = Constant() {r=0.5}
  MaterialExpressionAdd_7 = Add(
   A=Multiply(A=NamedRerouteUsage(), B=ScalarParameter() {defaultValue=1, group=Agriculture Flipbook Face, parameterName=Eyes_Scale, sortPriority=6})
   B=AppendVector(
    A=ScalarParameter() {group=Agriculture Flipbook Face, parameterName=Eyes_UVShiftX, sortPriority=6}
    B=ScalarParameter() {group=Agriculture Flipbook Face, parameterName=Eyes_UVShiftY, sortPriority=6}
   )
  )
  n41 = MaterialFunctionCall(Float2=@MaterialExpressionAdd_7) {materialFunction=&BreakOutFloat2Components}
  MaterialExpressionScalarParameter_13 = ScalarParameter() {defaultValue=3, group=Agriculture Flipbook Face, parameterName=Eyes_SubImages, sortPriority=6}
  MaterialExpressionNamedRerouteUsage_47 = NamedRerouteUsage()
  n42 = Custom(uv=@MaterialExpressionAdd_7, size=@MaterialExpressionScalarParameter_13, frame=@MaterialExpressionNamedRerouteUsage_47) {
   code="float size2 = size-0.00001f;\r\nreturn uv/size2 + float2(floor(fmod(frame,size2)), floor(frame/size2))/size2;"
   description=SubUV
   outputType=CMOT_Float2
   inputs [
    {
     inputName=uv
     input {expression=&self:MaterialExpressionAdd_7}
    }
    {
     inputName=size
     input {expression=&self:MaterialExpressionScalarParameter_13}
    }
    {
     inputName=frame
     input {expression=&self:MaterialExpressionNamedRerouteUsage_47}
    }
   ]
  }
  MaterialExpressionNamedRerouteUsage_51 = NamedRerouteUsage()
  n43 = Custom(uv=@MaterialExpressionAdd_7, size=@MaterialExpressionScalarParameter_13, frame=@MaterialExpressionNamedRerouteUsage_51) {
   code="float size2 = size-0.00001f;\r\nreturn uv/size2 + float2(floor(fmod(frame,size2)), floor(frame/size2))/size2;"
   description=SubUV
   outputType=CMOT_Float2
   inputs [
    {
     inputName=uv
     input {expression=&self:MaterialExpressionAdd_7}
    }
    {
     inputName=size
     input {expression=&self:MaterialExpressionScalarParameter_13}
    }
    {
     inputName=frame
     input {expression=&self:MaterialExpressionNamedRerouteUsage_51}
    }
   ]
  }
  n44 = TextureObjectParameter() {
   group=Agriculture Flipbook Face
   parameterName=Eyes_Flipbook
   texture=&T_ChiveFlake_Head_Flipbook_Eyes
  }
  n45 = TextureSample(UVs=LinearInterpolate(A=@n42, B=@n43, Alpha=NamedRerouteUsage()), Tex=@n44) {
   mipValueMode=TMVM_MipBias
   samplerSource=SSM_Clamp_WorldGroupSettings
   texture=&T_ChiveFlake_Head_Flipbook_Eyes
  }
  n46 = Multiply(
   A=Multiply(A=SphereMask(A=@n40, B=@n41.R) {attenuationRadius=0.5}, B=SphereMask(A=@n41.G, B=@n40) {attenuationRadius=0.5})
   B=MaterialFunctionCall(DF=@n45.A) {materialFunction=&MF_DFLine}.Result
  )
  MaterialExpressionAdd_10 = Add(
   A=Multiply(A=NamedRerouteUsage(), B=ScalarParameter() {defaultValue=1, group=Agriculture Flipbook Face, parameterName=Brows_Scale, sortPriority=5})
   B=AppendVector(
    A=ScalarParameter() {group=Agriculture Flipbook Face, parameterName=Brows_UVShiftX, sortPriority=5}
    B=ScalarParameter() {group=Agriculture Flipbook Face, parameterName=Brows_UVShiftY, sortPriority=5}
   )
  )
  MaterialExpressionScalarParameter_22 = ScalarParameter() {defaultValue=3, group=Agriculture Flipbook Face, parameterName=Brows_SubImages, sortPriority=5}
  MaterialExpressionNamedRerouteUsage_55 = NamedRerouteUsage()
  n47 = Custom(uv=@MaterialExpressionAdd_10, size=@MaterialExpressionScalarParameter_22, frame=@MaterialExpressionNamedRerouteUsage_55) {
   code="float size2 = size-0.00001f;\r\nreturn uv/size2 + float2(floor(fmod(frame,size2)), floor(frame/size2))/size2;"
   description=SubUV
   outputType=CMOT_Float2
   inputs [
    {
     inputName=uv
     input {expression=&self:MaterialExpressionAdd_10}
    }
    {
     inputName=size
     input {expression=&self:MaterialExpressionScalarParameter_22}
    }
    {
     inputName=frame
     input {expression=&self:MaterialExpressionNamedRerouteUsage_55}
    }
   ]
  }
  MaterialExpressionNamedRerouteUsage_58 = NamedRerouteUsage()
  n48 = Custom(uv=@MaterialExpressionAdd_10, size=@MaterialExpressionScalarParameter_22, frame=@MaterialExpressionNamedRerouteUsage_58) {
   code="float size2 = size-0.00001f;\r\nreturn uv/size2 + float2(floor(fmod(frame,size2)), floor(frame/size2))/size2;"
   description=SubUV
   outputType=CMOT_Float2
   inputs [
    {
     inputName=uv
     input {expression=&self:MaterialExpressionAdd_10}
    }
    {
     inputName=size
     input {expression=&self:MaterialExpressionScalarParameter_22}
    }
    {
     inputName=frame
     input {expression=&self:MaterialExpressionNamedRerouteUsage_58}
    }
   ]
  }
  n49 = LinearInterpolate(A=@n47, B=@n48, Alpha=NamedRerouteUsage())
  n50 = TextureObjectParameter() {
   group=Agriculture Flipbook Face
   parameterName=Brows_Flipbook
   samplerType=SAMPLERTYPE_Masks
   texture=/BRCosmetics/Characters/Player/Female/Medium/Bodies/F_MED_ChiveFlake/Textures/T_ChiveFlake_Head_Flipbook_Brows
  }
  n51 = TextureSample(UVs=@n49, Tex=@n50) {mipValueMode=TMVM_MipBias, samplerSource=SSM_Clamp_WorldGroupSettings, samplerType=SAMPLERTYPE_Masks}
  n52 = Subtract(A=@n46, B=@n51.R)
  NamedRerouteDeclaration(None=StaticSwitchParameter(True=Saturate(None=@n52), False=@n46) {group=Agriculture Flipbook Face, parameterName=UseNoAlphaBrowsTex}) {name=Alpha_Eyes, nodeColor=<rgba 1 0 0.1875 1>}
  NamedRerouteDeclaration(None=@n45.RGB) {name=Color_Eyes, nodeColor=<rgba 0 0.5078 1 1>}
  n53 = VertexInterpolator(VS=LocalPosition() {localOrigin=InstancePreSkinning}.XYZ)
  NamedRerouteDeclaration(None=Step(X=ComponentMask(None=@n53.PS) {r=true})) {name=LeftMask, nodeColor=<rgba 0 0.0391 1 1>}
  n54 = NamedRerouteUsage()
  n55 = VectorParameter() {group=TwinkleLoop Head, parameterName=TL_Pupil_R_Center, defaultValue=<rgba 0.55 0.53 0 1>}
  n56 = Add(
   A=ScalarParameter() {group=1 - Eye Animation Poses, parameterName=look_left_pose, sliderMax=1, sortPriority=11}
   B=Multiply(A=ScalarParameter() {group=1 - Eye Animation Poses, parameterName=look_right_pose, sliderMax=1, sortPriority=10}) {constB=-1}
  )
  n57 = Add(
   A=ScalarParameter() {group=1 - Eye Animation Poses, parameterName=look_down_pose, sliderMax=1, sortPriority=13}
   B=Multiply(A=ScalarParameter() {group=1 - Eye Animation Poses, parameterName=look_up_pose, sliderMax=1, sortPriority=12}) {constB=-1}
  )
  n58 = AppendVector(
   A=Multiply(
    A=Clamp(None=@n56) {minDefault=-1}
    B=ScalarParameter() {defaultValue=0.2, group=TwinkleLoop Head, parameterName=TL_EyeHorizontal_Threshold, sliderMax=1}
   )
   B=Multiply(
    A=Clamp(None=@n57) {minDefault=-1}
    B=ScalarParameter() {defaultValue=0.2, group=TwinkleLoop Head, parameterName=TL_EyeVertical_Threshold, sliderMax=1}
   )
  )
  n59 = Add(A=ComponentMask(None=@n55.RGB) {g=true, r=true}, B=@n58)
  n60 = ScalarParameter() {defaultValue=0.34, group=TwinkleLoop Head, parameterName=TL_Pupil_Radius}
  n61 = VectorParameter() {group=TwinkleLoop Head, parameterName=TL_Pupil_L_Center, defaultValue=<rgba 0.45 0.53 0 1>}
  n62 = Add(A=@n58, B=ComponentMask(None=@n61.RGB) {g=true, r=true})
  n63 = Multiply(
   A=Constant3Vector() {constant=<rgba 0.9 0.9 0.9 0>}
   B=LinearInterpolate(
    A=SphereMask(A=@n54, B=@n59, Radius=@n60) {attenuationRadius=0.05}
    B=SphereMask(A=@n54, B=@n62, Radius=@n60) {attenuationRadius=0.05}
    Alpha=NamedRerouteUsage()
   )
  )
  n64 = VectorParameter() {group=TwinkleLoop Head, parameterName=TL_Eyelidcolor, defaultValue=<rgba 1 0.4918 0.0312 1>}
  n65 = NamedRerouteUsage()
  n66 = LinearInterpolate(
   A=NamedRerouteUsage()
   B=LinearInterpolate(A=LinearInterpolate(A=@n63, B=@n64.RGB, Alpha=@n65), Alpha=NamedRerouteUsage()) {constB=0}
   Alpha=NamedRerouteUsage()
  )
  n67 = NamedRerouteUsage()
  n68 = Multiply(A=NamedRerouteUsage(), B=NamedRerouteUsage())
  n69 = NamedRerouteUsage()
  n70 = NamedRerouteUsage()
  NamedRerouteDeclaration(None=LinearInterpolate(A=LinearInterpolate(A=@n66, B=@n67, Alpha=@n68), B=NamedRerouteUsage(), Alpha=Multiply(A=@n69, B=@n70))) {name=TL_Color_OUT, nodeColor=<rgba 0.125 1 0 1>}
  n71 = Constant() {r=0.5}
  n72 = MaterialFunctionCall(Float2=@MaterialExpressionAdd_10) {materialFunction=&BreakOutFloat2Components}
  n73 = SphereMask(A=@n71, B=@n72.R) {attenuationRadius=0.5}
  n74 = SphereMask(A=@n72.G, B=@n71) {attenuationRadius=0.5}
  NamedRerouteDeclaration(None=Multiply(
    A=Multiply(A=@n73, B=@n74)
    B=MaterialFunctionCall(DF=@n51.G) {materialFunction=&MF_DFLine}.Result
   )) {name=Alpha_Brows, nodeColor=<rgba 1 0 0.1875 1>}
  n75 = Constant3Vector()
  NamedRerouteDeclaration(None=@n75) {name=Color_Brows, nodeColor=<rgba 0 0.5078 1 1>}
  n76 = MaterialFunctionCall(Float2=TextureCoordinate() {coordinateIndex=1}) {materialFunction=&BreakOutFloat2Components}
  n77 = Add(A=@n76.R) {constB=-1}
  NamedRerouteDeclaration(None=Saturate(None=Floor(None=@n77))) {name=TL_MouthBrows_Mask, nodeColor=<rgba 0.3359 0 1 1>}
  NamedRerouteDeclaration() {name=Debug, nodeColor=<rgba 1 0 0.4219 1>}
  n78 = Floor(None=Add(A=TextureCoordinate() {coordinateIndex=1}) {constB=0})
  n79 = MaterialFunctionCall(Float2=@n78) {materialFunction=&BreakOutFloat2Components}
  n80 = Max(A=Add(A=@n79.R) {constB=-1}, B=@n79.G) {constB=0}
  n81 = MaterialFunctionCall(Float2=Multiply(A=@n78) {constB=-1}) {materialFunction=&BreakOutFloat2Components}
  n82 = Max(A=@n81.R, B=@n81.R) {constB=0}
  NamedRerouteDeclaration(None=OneMinus(None=Max(A=@n80, B=@n82) {constB=0})) {name=Mask_EyesUDIM, nodeColor=<rgba 1 0.2578 0 1>}
  n83 = ScalarParameter() {group=Agriculture Flipbook Face Pose, parameterName=flipbook_face_R_eye_index, sortPriority=6}
  n84 = ScalarParameter() {group=Agriculture Flipbook Face Pose, parameterName=flipbook_face_L_eye_index, sortPriority=6}
  n85 = SphereMask(A=Max(A=@n83, B=@n84), B=Constant() {r=7}) {attenuationRadius=0.02}
  NamedRerouteDeclaration(None=LinearInterpolate(A=@n83, Alpha=@n85) {constB=7}) {name=R_Eye_Index, nodeColor=<rgba 0.0547 1 0 1>}
  NamedRerouteDeclaration(None=LinearInterpolate(A=@n84, Alpha=@n85) {constB=7}) {name=L_Eye_Index, nodeColor=<rgba 0.0547 1 0 1>}
  n86 = NamedRerouteUsage()
  NamedRerouteDeclaration(None=Multiply(A=LinearInterpolate(A=@n16, B=@n19, Alpha=@n86), B=NamedRerouteUsage())) {name=TL_EyesHide, nodeColor=<rgba 0.0078 0 1 1>}
  NamedRerouteDeclaration(None=Frac(None=TextureCoordinate() {coordinateIndex=1})) {name=TL_UVs, nodeColor=<rgba 0.5469 1 0 1>}
  n87 = ComponentMask(None=NamedRerouteUsage()) {g=true}
  n88 = ScalarParameter() {defaultValue=0.032, parameterName=TL_EyelidLineThickness}
  n89 = Multiply(A=@n88) {constB=0.5}
  n90 = Add(A=@n87, B=@n89) {constB=0}
  n91 = LinearInterpolate(A=NamedRerouteUsage(), B=NamedRerouteUsage(), Alpha=NamedRerouteUsage())
  n92 = Clamp(None=ScalarParameter() {group=1 - Eye Animation Poses, parameterName=R_blink_pose, sliderMax=1, sortPriority=2})
  n93 = Multiply(A=NamedRerouteUsage(), B=@n92)
  n94 = Clamp(None=ScalarParameter() {group=1 - Eye Animation Poses, parameterName=L_blink_pose, sliderMax=1, sortPriority=1})
  n95 = Multiply(A=@n94, B=NamedRerouteUsage())
  n96 = NamedRerouteUsage()
  n97 = Saturate(None=Power(Base=LinearInterpolate(A=@n93, B=@n95, Alpha=@n96), Exp=Constant() {r=1.2}))
  n98 = LinearInterpolate(B=@n91, Alpha=@n97)
  n99 = Step(Y=@n90, X=@n98) {constX=0}
  n100 = LinearInterpolate(B=@n91, Alpha=@n97) {constA=1}
  n101 = Subtract(A=@n87, B=@n89)
  n102 = Step(Y=@n100, X=@n101) {constX=0.5}
  NamedRerouteDeclaration(None=Saturate(None=Add(A=@n99, B=@n102))) {name=TL_Eyelids_Mask, nodeColor=<rgba 0 0.1562 1 1>}
  n103 = LinearInterpolate(A=NamedRerouteUsage(), B=NamedRerouteUsage(), Alpha=NamedRerouteUsage())
  n104 = NamedRerouteUsage()
  n105 = NamedRerouteUsage()
  NamedRerouteDeclaration(None=LinearInterpolate(A=LinearInterpolate(A=@n103, B=@n104, Alpha=@n105), B=NamedRerouteUsage(), Alpha=NamedRerouteUsage())) {name=Color_OUT, nodeColor=<rgba 0.3359 1 0 1>}
  n106 = Subtract(A=Step(Y=@n90, X=Add(A=@n98, B=@n88) {constB=0.025}), B=@n99) {desc=right upper lid}
  n107 = Subtract(A=Step(Y=Subtract(A=@n100, B=@n88), X=@n101), B=@n102) {desc=right lower lid}
  NamedRerouteDeclaration(None=Saturate(None=Max(A=@n106, B=@n107))) {name=TL_EyelidsLines_Mask, nodeColor=<rgba 1 0.7969 0 1>}
  NamedRerouteDeclaration(None=@n55.G) {name=TL_Pupil_R_CenterY, nodeColor=<rgba 1 0.1406 0 1>}
  NamedRerouteDeclaration(None=@n61.G) {name=TL_Pupil_L_CenterY, nodeColor=<rgba 1 0.5166 0.4375 1>}
  n108 = ScalarParameter() {group=Agriculture Flipbook Face Pose, parameterName=flipbook_face_R_brow_index, sortPriority=5}
  NamedRerouteDeclaration(None=@n108) {name=R_Brow_Index, nodeColor=<rgba 1 0.3083 0 1>}
  n109 = ScalarParameter() {group=Agriculture Flipbook Face Pose, parameterName=flipbook_face_L_brow_index, sortPriority=5}
  NamedRerouteDeclaration(None=@n109) {name=L_Brow_Index, nodeColor=<rgba 1 0.1082 0.0795 1>}
  NamedRerouteDeclaration(None=@n51.B) {name=EyelidMaskFromBrow, nodeColor=<rgba 1 0.3047 0 1>}
  n110 = Max(A=SphereMask(A=@n108, B=Constant() {r=3}) {attenuationRadius=0.02}, B=SphereMask(A=@n108, B=Constant() {r=4}) {attenuationRadius=0.02})
  n111 = SphereMask(A=@n108, B=Constant() {r=6}) {attenuationRadius=0.02}
  NamedRerouteDeclaration(None=OneMinus(None=Max(A=@n110, B=@n111))) {name=R_Brow_EyeBlink, nodeColor=<rgba 0.7344 0 1 1>}
  n112 = Max(A=SphereMask(A=@n109, B=Constant() {r=3}) {attenuationRadius=0.02}, B=SphereMask(A=@n109, B=Constant() {r=4}) {attenuationRadius=0.02})
  n113 = SphereMask(A=@n109, B=Constant() {r=6}) {attenuationRadius=0.02}
  NamedRerouteDeclaration(None=OneMinus(None=Max(A=@n112, B=@n113))) {name=L_Brow_EyeBlink, nodeColor=<rgba 0.7344 0 1 1>}
  n114 = OneMinus(None=Sign(None=@n92))
  n115 = NamedRerouteUsage()
  n116 = OneMinus(None=Sign(None=@n94))
  n117 = NamedRerouteUsage()
  NamedRerouteDeclaration(None=Add(A=Multiply(A=@n114, B=@n115), B=Multiply(A=@n116, B=@n117))) {name=TL_DFL_Toggle, nodeColor=<rgba 0 0.9297 1 1>}
  n118 = TextureObjectParameter() {
   group=Agriculture Flipbook Face
   parameterName=ChiveFlakeBackpack_Flipbook
   texture=/BRCosmetics/Accessories/FORT_Backpacks/Backpack_F_MED_ChiveFlake/Textures/T_ChiveFlake_Backpack_Flipbook
  }
  n119 = TextureSample!(UVs=NamedRerouteUsage(), Tex=@n118, Apply View MipBias)
  n120 = MaterialFunctionCall(DF=@n119.A) {materialFunction=&MF_DFLine}
  NamedRerouteDeclaration(None=LinearInterpolate(A=NamedRerouteUsage(), B=@n119.RGB, Alpha=@n120.Result)) {name=CFB_Color_OUT, nodeColor=<rgba 1 0 0.6562 1>}
  n121 = ScalarParameter() {group=Agriculture Flipbook Face Pose, parameterName=flipbook_face_mouth_index, sortPriority=7}
  n122 = SphereMask(A=@n121, B=Constant()) {attenuationRadius=0.02}
  n123 = ScalarParameter() {group=Agriculture Flipbook Face Pose, parameterName=flipbook_face_is_idling}
  NamedRerouteDeclaration(None=LinearInterpolate(A=@n121, B=LinearInterpolate(A=@n121, Alpha=@n122) {constB=11}, Alpha=Floor(None=@n123))) {name=Mouth_Index, nodeColor=<rgba 0 1 0.7656 1>}
  NamedRerouteDeclaration(None=@n120.Result) {name=CFB_Alpha_OUT, nodeColor=<rgba 0.4531 1 0 1>}
  n124 = NamedRerouteUsage()
  n125 = Constant()
  n126 = If(
   A=DotProduct(A=Constant3Vector() {constant=<rgba 1 1 1 0>}, B=VertexColor())
   B=Constant() {r=3}
   "A > B"=@n125
   "A == B"=@n125
   "A < B"=Constant() {r=1}
  )
  n127 = Floor(None=ScalarParameter() {group=ChiveFlakeBackpack Face, parameterName=Expression_Eyes, sliderMax=3})
  n128 = MaterialFunctionCall!(1=Constant() {r=0.1}, 2=Constant() {r=0.15}, 3=Constant() {r=0.2}, 4, A=Divide(A=Clamp(None=@n127) {maxDefault=2})) {materialFunction=/Engine/Functions/Engine_MaterialFunctions02/Lerp_Multiple_Float}
  n129 = VertexColor()
  n130 = ScalarParameter() {group=ChiveFlakeBackpack Face, parameterName=Expression_Mouth, sliderMax=1}
  n131 = LinearInterpolate(A=Constant() {r=0.2}, B=Constant() {r=0.5}, Alpha=Clamp(None=Floor(None=@n130)))
  n132 = Constant() {r=0.15}
  n133 = Saturate(None=Add(
    A=SphereMask(A=@n128.Lerp 3 Inputs, B=@n129.G) {attenuationRadius=0.02}
    B=SphereMask(A=@n131, B=@n129.B, Radius=@n132) {attenuationRadius=0.1}
   ))
  n134 = Saturate(None=Round(None=Subtract(A=@n126, B=@n133)))
  NamedRerouteDeclaration(None=PreviousFrameSwitch(Current Frame=Multiply(A=@n124, B=@n134), Previous Frame=Constant())) {name=CFB_WPO_OUT, nodeColor=<rgba 0 0.6484 1 1>}
  n135 = WorldPosition()
  n136 = ObjectPositionWS()
  NamedRerouteDeclaration(None=Multiply(A=Subtract(A=@n135.XYZ, B=@n136)) {constB=-1}) {name=Vertex Color Hide, nodeColor=<rgba 0.6875 0 1 1>}
  n137 = VectorParameter() {group=TwinkleLoopBackpack Face, parameterName=TLB_ExpressionColor, defaultValue=<rgba 0.005 0.005 0.005 1>}
  n138 = NamedRerouteUsage()
  n139 = Constant() {r=0.5}
  n140 = MaterialFunctionCall(Float2=@n29.UVs) {materialFunction=&BreakOutFloat2Components}
  n141 = Multiply(A=SphereMask(A=@n139, B=@n140.R) {attenuationRadius=0.5}, B=SphereMask(A=@n140.G, B=@n139) {attenuationRadius=0.5})
  NamedRerouteDeclaration(None=LinearInterpolate(A=NamedRerouteUsage(), B=@n137.RGB, Alpha=Multiply(A=@n138, B=@n141))) {name=TLB_Color_OUT, nodeColor=<rgba 1 0 0.6562 1>}
  n142 = NamedRerouteDeclaration(None=Min(A=Ceil(None=@n76.G), B=NamedRerouteUsage())) {name=TL_Eyes_Mask, nodeColor=<rgba 1 0 0.8906 1>}
  n143 = Multiply(A=OneMinus(None=Floor(None=@n76.R)), B=@n142) {desc=LeftEye}
  NamedRerouteDeclaration(None=Subtract(A=@n142, B=NamedRerouteDeclaration(None=@n143) {name=TL_RightEye_Mask, nodeColor=<rgba 0.0293 0.3125 0.0353 1>})) {name=TL_LeftEye_Mask, nodeColor=<rgba 1 0 0.6797 1>}
  NamedRerouteUsage()
  n144 = ScalarParameter() {defaultValue=0.1, desc="LineThickness (Near)", group=TwinkleLoopBackpack Face, parameterName=TLB_LineThickness, sliderMax=1, sortPriority=3}
  n145 = ScalarParameter() {defaultValue=0.7, group=TwinkleLoopBackpack Face, parameterName=TLB_LineThickness_Far, sliderMax=2, sortPriority=3}
  n146 = NamedRerouteUsage()
  n147 = Subtract(A=OneMinus(None=LinearInterpolate(A=@n144, B=@n145, Alpha=@n146)), B=@n31)
  n148 = DDX(Value=@n147)
  n149 = DDY(Value=@n147)
  n150 = Add(
   A=Length(None=AppendVector(A=@n148, B=@n149)) {is=class MaterialExpressionLength}
   B=LinearInterpolate(Alpha=ScalarParameter() {defaultValue=0.9, group=TwinkleLoopBackpack Face, parameterName=TLB_LineSoftness}) {constA=0.5, constB=0}
  ) {constB=0}
  Saturate(None=Subtract(B=Divide(A=@n147, B=@n150)) {constA=0.5})
  StaticSwitchParameter(True=@n13, False=Constant()) {defaultValue=true, group=Agriculture Flipbook Face, parameterName=UseFlipbookFace, sortPriority=-1}
  StaticSwitchParameter(True=@n6, False=Constant()) {defaultValue=true, group=Agriculture Flipbook Face, parameterName=UseFlipbookFace, sortPriority=-1}
  n151 = TextureSample!(UVs=@n49, Tex=@n50, Bias, Apply View MipBias) {
   mipValueMode=TMVM_MipBias
   samplerSource=SSM_Clamp_WorldGroupSettings
   samplerType=SAMPLERTYPE_Masks
   texture=&T_ChiveFlake_Head_Flipbook_Eyes
  }
  StaticSwitchParameter(True=@n51.G, False=@n151.A) {group=Agriculture Flipbook Face, parameterName=UseNoAlphaBrowsTex}
  StaticSwitchParameter(True=@n75, False=@n151.RGB) {group=Agriculture Flipbook Face, parameterName=UseNoAlphaBrowsTex}
 ]}
