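# Parent cloth-shading material for the M_LRG_PuzzleShed character's anime-style flipbook face.
# Shared anime Pre and Calculations functions feed toon lighting, with normals optionally blended toward a spherical or cylindrical shape around a pivot.
# MF_Agriculture_FlipbookFace supplies color, opacity, WPO and lighting mask to the Post function, with shaded color derived by darkening.
# Static switch UseGadgetCajunTaco replaces both colors with a time-driven, four-column emote-prop sprite sheet on UV1.
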Material /BRCosmetics/Characters/Player/Male/Large/Bodies/M_LRG_PuzzleShed/Materials/M_Agriculture_FlipbookFace_PARENT
graph {
 properties {
  bAutomaticallySetUsageInEditor=false
  bUseMaterialAttributes=true
  bUsedWithMeshDeformer=true
  bUsedWithMorphTargets=true
  bUsedWithSkeletalMesh=true
  bUsedWithStaticLighting=true
  pixelDepthOffsetMode=PDOM_Legacy
  shadingModel=MSM_Cloth
  usedShadingModels=Cloth
 }
 nodes [
  n1 = MaterialFunctionCall!(CustomUV) {materialFunction=/Game/Global/Materials/Character_Parent_2023/Anime_Parent_2023/MaterialFunctions/MFC_AnimeShader_2023_Pre}
  n2 = Transform(None=@n1.Normal)
  n3 = Subtract(
   A=LocalPosition() {localOrigin=InstancePreSkinning}.XYZ
   B=VectorParameter() {group=Toon_Shading, parameterName=NormalSphere_Location, defaultValue=<rgba 0 0 162 1>}.RGB
  )
  n4 = LinearInterpolate(
   A=@n3
   B=Multiply(A=@n3, B=Constant3Vector() {constant=<rgba 1 1 0 0>})
   Alpha=ScalarParameter() {group=Toon_Shading, parameterName=NormalSphereCylindrical_Blend, sliderMax=1}
  )
  n5 = VertexInterpolator(VS=Transform(None=Normalize(VectorInput=@n4)) {transformSourceType=TRANSFORMSOURCE_Local})
  n6 = ScalarParameter() {group=Toon_Shading, parameterName=NormalSphere_Blend, sliderMax=1}
  n7 = ChannelMaskParameter(None=@n1.FX) {group=Toon_Shading, parameterName=ThresholdSphereNormalMask}
  n8 = ScalarParameter() {group=Toon_Shading, parameterName=NormalSphereMask, sliderMax=1}
  n9 = MaterialFunctionCall!(
   CustomLightVector
   OverrideNormal_WS=LinearInterpolate(A=@n2, B=LinearInterpolate(A=@n2, B=@n5.PS, Alpha=@n6) {constAlpha=0}, Alpha=LinearInterpolate(B=@n7.R, Alpha=@n8) {constA=1})
   STT=@n1.STT
   Normal_Map=@n1.STT
   DFL=@n1.STT
   FX=@n1.STT
  ) {materialFunction=/Game/Global/Materials/Character_Parent_2023/Anime_Parent_2023/MaterialFunctions/MFC_AnimeShader_2023_Calculations}
  n10 = MaterialFunctionCall!(ColorIn=@n1.Color_Lit, WPOIn, DFL_Mask=@n9.DFL_Mask) {materialFunction=&/MF_Agriculture_FlipbookFace}
  n11 = Multiply(
   A=TextureCoordinate() {coordinateIndex=1}
   B=ScalarParameter() {defaultValue=1, group=Gadget CajunTaco, parameterName=ScreenUV_Scale, sortPriority=6}
  )
  n12 = NamedRerouteUsage()
  n13 = Constant() {r=4}
  n14 = Divide(A=NamedRerouteUsage(), B=Constant() {r=4})
  n15 = Multiply(
   A=MaterialFunctionCall(X=Fmod(A=@n12, B=@n13), Y=Floor(None=@n14)) {materialFunction=&MakeFloat2}.Result
   B=Divide(A=Constant2Vector() {g=162, r=235}, B=Constant() {r=1024})
  )
  n16 = ScalarParameter() {group=Gadget CajunTaco, parameterName=ScreenUV_ShiftX, sortPriority=6}
  n17 = ScalarParameter() {group=Gadget CajunTaco, parameterName=ScreenUV_ShiftY, sortPriority=6}
  n18 = TextureSample(UVs=Add(A=Add(A=@n11, B=@n15), B=AppendVector(A=@n16, B=@n17))) {texture=/BRCosmetics/Animation/Props/Gadgets_CajunTaco/Textures/T_CajunTaco_Emote_Prop_SpriteSheet}
  n19 = ScalarParameter() {defaultValue=0.5, group=01_Anime Maps, parameterName=Dynamic_ColorShaded_Darkening, sliderMax=1, sortPriority=11}
  MaterialFunctionCall!(
   Shading_Gradient=@n9.Shading_Gradient
   Threshold_Gradient=@n9.Shading_Gradient
   Specular_Gradient=@n9.Shading_Gradient
   Sheen_Mask=@n9.Shading_Gradient
   Thickness=@n9.Shading_Gradient
   DFL_Mask=NamedRerouteDeclaration(None=@n10.DFL_Mask) {name=DFL_Mask_OUT, nodeColor=<rgba 0 0.5078 1 1>}
   FX=@n9.Shading_Gradient
   Metallic_Gradient=@n9.Shading_Gradient
   RimLight_Gradient=@n9.Shading_Gradient
   NormalBuffer=@n9.Shading_Gradient
   --------
   CINE_LightSources=@n9.Shading_Gradient
   CINE_LightColor=@n9.Shading_Gradient
   ---------
   Color_Lit=StaticSwitchParameter(True=@n18.RGB, False=@n10.Color) {group=Gadget CajunTaco, parameterName=UseGadgetCajunTaco}
   Color_Shaded=StaticSwitchParameter(True=@n18.RGB, False=Multiply(A=@n10.Color, B=@n19)) {group=Gadget CajunTaco, parameterName=UseGadgetCajunTaco}
   SRM=@n1.SRM
   Emissive=@n1.SRM
   DFL_Color=@n1.SRM
   Specular_Color=@n1.SRM
   RimLight_Color=@n1.SRM
   OpacityMaskIn=NamedRerouteDeclaration(None=@n10.Alpha) {name=OpacityMask_OUT, nodeColor=<rgba 0 1 0.4375 1>}
   WPOIn=NamedRerouteDeclaration(None=@n10.WPO) {name=WPO_OUT, nodeColor=<rgba 0 1 0.7891 1>}
  ) {materialFunction=&/M_Agriculture_FlipbookFace_Post}
  n20 = Frac(None=Multiply(A=Time(), B=ScalarParameter() {defaultValue=0.25, group=Gadget CajunTaco, parameterName=Screen_PlaySpeed}))
  n21 = ScalarParameter() {defaultValue=17, group=Gadget CajunTaco, parameterName=Screen_TotalFrames}
  NamedRerouteDeclaration(None=Round(None=Multiply(A=@n20, B=@n21))) {name=CurrFrame, nodeColor=<rgba 1 0.2812 0 1>}
 ]
}
