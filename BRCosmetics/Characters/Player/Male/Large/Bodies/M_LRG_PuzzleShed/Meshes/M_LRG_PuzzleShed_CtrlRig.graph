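# Control Rig for the M_LRG_PuzzleShed large male player body cosmetic that drives deformation bones from the body skeleton.
# Setup captures bind and offset corrective poses for limb spline controls.
# Spherical pose readers turn limb angles into alphas that blend those poses.
# Constrained controls feed stretchy B-spline IK on arm and leg chains, with variable-weighted backpack, DBNO and arm adjustments, groin/hip correctives and neck/head rotation damping.
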ControlRigBlueprint /BRCosmetics/Characters/Player/Male/Large/Bodies/M_LRG_PuzzleShed/Meshes/M_LRG_PuzzleShed_CtrlRig
graphs [
 ContainedGraph {nodes [
   n1 = RigVMFunctionEntryNode()
   SequenceAggregate_1 = &RigVMFunction_Sequence(ExecuteContext=@n1.ExecuteContext)
   SequenceAggregate_1_1 = &RigVMFunction_Sequence(ExecuteContext=@SequenceAggregate_1.B)
   SequenceAggregate_1_2 = &RigVMFunction_Sequence(ExecuteContext=@SequenceAggregate_1_1.B)
   SequenceAggregate_1_3 = &RigVMFunction_Sequence(ExecuteContext=@SequenceAggregate_1_2.B)
   SequenceAggregate_1_5 = &RigVMFunction_Sequence(ExecuteContext=@SequenceAggregate_1_3.B)
   SequenceAggregate_1_6 = &RigVMFunction_Sequence(ExecuteContext=@SequenceAggregate_1_5.B)
   SequenceAggregate_1_4 = &RigVMFunction_Sequence(ExecuteContext=@SequenceAggregate_1_6.B)
   RigVMFunctionReturnNode(
    A=@SequenceAggregate_1.A
    B=@SequenceAggregate_1_1.A
    C=@SequenceAggregate_1_2.A
    D=@SequenceAggregate_1_3.A
    H=@SequenceAggregate_1_4.A
    I=@SequenceAggregate_1_4.B
    F=@SequenceAggregate_1_5.A
    G=@SequenceAggregate_1_6.A
   )
  ]}
 ContainedGraph {nodes [
   n1 = RigVMFunctionEntryNode()
   SequenceAggregate_3 = &RigVMFunction_Sequence(ExecuteContext=@n1.ExecuteContext)
   SequenceAggregate_3_1 = &RigVMFunction_Sequence(ExecuteContext=@SequenceAggregate_3.B)
   SequenceAggregate_3_2 = &RigVMFunction_Sequence(ExecuteContext=@SequenceAggregate_3_1.B)
   RigVMFunctionReturnNode(A=@SequenceAggregate_3.A, B=@SequenceAggregate_3_1.A, C=@SequenceAggregate_3_2.A, D=@SequenceAggregate_3_2.B)
  ]}
 ContainedGraph {nodes [
   n1 = RigVMFunctionEntryNode()
   SequenceAggregate_2 = &RigVMFunction_Sequence(ExecuteContext=@n1.ExecuteContext)
   SequenceAggregate_2_1 = &RigVMFunction_Sequence(ExecuteContext=@SequenceAggregate_2.B)
   SequenceAggregate_2_2 = &RigVMFunction_Sequence(ExecuteContext=@SequenceAggregate_2_1.B)
   RigVMFunctionReturnNode(A=@SequenceAggregate_2.A, B=@SequenceAggregate_2_1.A, C=@SequenceAggregate_2_2.A, D=@SequenceAggregate_2_2.B)
  ]}
 ContainedGraph {nodes [
   n1 = RigVMFunctionEntryNode()
   SequenceAggregate_2_1 = &RigVMFunction_Sequence(ExecuteContext=@n1.ExecuteContext)
   SequenceAggregate_2_1_1 = &RigVMFunction_Sequence(ExecuteContext=@SequenceAggregate_2_1.B)
   SequenceAggregate_2_1_2 = &RigVMFunction_Sequence(ExecuteContext=@SequenceAggregate_2_1_1.B)
   RigVMFunctionReturnNode(A=@SequenceAggregate_2_1.A, B=@SequenceAggregate_2_1_1.A, C=@SequenceAggregate_2_1_2.A, D=@SequenceAggregate_2_1_2.B)
  ]}
 ContainedGraph {nodes [
   n1 = RigVMFunctionEntryNode()
   SequenceAggregate = &RigVMFunction_Sequence(ExecuteContext=@n1.ExecuteContext)
   SequenceAggregate_1 = &RigVMFunction_Sequence(ExecuteContext=@SequenceAggregate.B)
   SequenceAggregate_2 = &RigVMFunction_Sequence(ExecuteContext=@SequenceAggregate_1.B)
   RigVMFunctionReturnNode(A=@SequenceAggregate.A, B=@SequenceAggregate_1.A, C=@SequenceAggregate_2.A, D=@SequenceAggregate_2.B)
  ]}
 ContainedGraph {nodes [
   n1 = RigVMFunctionEntryNode()
   SequenceAggregate = &RigVMFunction_Sequence(ExecuteContext=@n1.ExecuteContext)
   SequenceAggregate_1 = &RigVMFunction_Sequence(ExecuteContext=@SequenceAggregate.B)
   SequenceAggregate_2 = &RigVMFunction_Sequence(ExecuteContext=@SequenceAggregate_1.B)
   RigVMFunctionReturnNode(A=@SequenceAggregate.A, B=@SequenceAggregate_1.A, C=@SequenceAggregate_2.A, D=@SequenceAggregate_2.B)
  ]}
 ContainedGraph {nodes [
   n1 = RigVMFunctionEntryNode()
   SequenceAggregate_1 = &RigVMFunction_Sequence(ExecuteContext=@n1.ExecuteContext)
   SequenceAggregate_1_1 = &RigVMFunction_Sequence(ExecuteContext=@SequenceAggregate_1.B)
   SequenceAggregate_1_2 = &RigVMFunction_Sequence(ExecuteContext=@SequenceAggregate_1_1.B)
   SequenceAggregate_1_3 = &RigVMFunction_Sequence(ExecuteContext=@SequenceAggregate_1_2.B)
   SequenceAggregate_1_4 = &RigVMFunction_Sequence(ExecuteContext=@SequenceAggregate_1_3.B)
   SequenceAggregate_1_5 = &RigVMFunction_Sequence(ExecuteContext=@SequenceAggregate_1_4.B)
   RigVMFunctionReturnNode(
    A=@SequenceAggregate_1.A
    B=@SequenceAggregate_1_1.A
    C=@SequenceAggregate_1_2.A
    D=@SequenceAggregate_1_3.A
    E=@SequenceAggregate_1_4.A
    F=@SequenceAggregate_1_5.A
    G=@SequenceAggregate_1_5.B
   )
  ]}
 RigVMFunctionLibrary {links [], nodes []}
 RigVMModel {nodes [
   RigUnit_BeginExecution = &RigUnit_BeginExecution()
   SequenceAggregate_1 = RigVMAggregateNode(ExecuteContext=@RigUnit_BeginExecution.ExecutePin)
   &RigUnit_ParentConstraint(
    AdvancedSettings=(InterpolationType=Shortest,RotationOrderForFilter=XZY)
    Child=(Type=Null,Name="dfrm_arm_spline_01_l_ctrl_null")
    Filter=(TranslationFilter=(bX=false,bY=false,bZ=false),RotationFilter=(bX=True,bY=True,bZ=True),ScaleFilter=(bX=True,bY=True,bZ=True))
    Parents=((Item=(Type=Bone,Name="upperarm_l"),Weight=0.3),(Item=(Type=Bone,Name="spine_05"),Weight=0.7))
    Weight=1
    ExecutePin=@SequenceAggregate_1.B
   )
   > &RigUnit_ParentConstraint(
    AdvancedSettings=(InterpolationType=Shortest,RotationOrderForFilter=XZY)
    Child=(Type=Null,Name="dfrm_arm_spline_05_l_ctrl_null")
    Parents=((Item=(Type=Control,Name="dfrm_arm_spline_01_l_ctrl"),Weight=0.3),(Item=(Type=Bone,Name="upperarm_l"),Weight=0.35),(Item=(Type=Bone,Name="lowerarm_l"),Weight=0.35))
    Weight=1
   )
   > &RigUnit_ParentConstraint(
    AdvancedSettings=(InterpolationType=Shortest,RotationOrderForFilter=XZY)
    Child=(Type=Null,Name="dfrm_arm_spline_03_l_ctrl_null")
    Parents=((Item=(Type=Control,Name="dfrm_arm_spline_01_l_ctrl"),Weight=0.7),(Item=(Type=Control,Name="dfrm_arm_spline_05_l_ctrl"),Weight=0.3))
    Weight=1
   )
   > &RigUnit_ParentConstraint(
    AdvancedSettings=(InterpolationType=Shortest,RotationOrderForFilter=XZY)
    Child=(Type=Null,Name="dfrm_arm_spline_09_l_ctrl_null")
    Filter=(TranslationFilter=(bX=True,bY=True,bZ=True),RotationFilter=(bX=True,bY=True,bZ=True),ScaleFilter=(bX=true,bY=true,bZ=true))
    Parents=((Item=(Type=Bone,Name="hand_l"),Weight=0.7),(Item=(Type=Bone,Name="lowerarm_l"),Weight=0.3))
    Weight=1
   )
   > &RigUnit_ParentConstraint(
    AdvancedSettings=(InterpolationType=Shortest,RotationOrderForFilter=XZY)
    Child=(Type=Null,Name="dfrm_arm_spline_07_l_ctrl_null")
    Parents=((Item=(Type=Control,Name="dfrm_arm_spline_05_l_ctrl"),Weight=0.5),(Item=(Type=Control,Name="dfrm_arm_spline_09_l_ctrl"),Weight=0.5))
    Weight=1
   )
   > &RigUnit_ModifyTransforms(
    ItemToModify=((Item=(Type=Control,Name="dfrm_arm_spline_01_l_ctrl"),Transform=(Rotation=(X=0,Y=0,Z=0,W=1),Translation=(X=4,Y=-1.5,Z=-2),Scale3D=(X=1,Y=1,Z=1))))
    Weight=RigVMVariableNode() {variable=UseBackpack_PuzzleShed}.Value
   )
   > &RigUnit_ModifyTransforms(
    ItemToModify=((Item=(Type=Control,Name="dfrm_arm_spline_03_l_ctrl"),Transform=(Rotation=(X=0,Y=0,Z=0,W=1),Translation=(X=0,Y=0,Z=10),Scale3D=(X=1,Y=1,Z=1))))
    Weight=RigVMVariableNode() {variable=spr_upper_l_DBNO}.Value
   )
   > &RigUnit_ModifyTransforms(
    ItemToModify="((Item=(Type=Control,Name=\"dfrm_arm_spline_03_l_ctrl\"),Transform=(Rotation=(X=0,Y=0,Z=0,W=1),Translation=(X=5,Y=0,Z=0),Scale3D=(X=1,Y=1,Z=1))),(Item=(Type=Control,Name=\"dfrm_arm_spline_05_l_ctrl\"),Transform=(Rotation=(X=0,Y=0,Z=0,W=1),Translation=(X=15,Y=0 ...(truncated)"
    Weight=RigVMVariableNode() {variable=spr_upper_l_up_alpha}.Value
   )
   RerouteNode_16 = RigVMDispatchNode(Values="((Type=Bone,Name=\"dfrm_clavicle_l\"),((),Type=Bone,Name=\"dfrm_arm_spline_02_l\"),((),Type=Bone,Name=\"dfrm_arm_spline_03_l\"),((),Type=Bone,Name=\"dfrm_arm_spline_04_l\"),((),Type=Bone,Name=\"dfrm_lowerarm_l\"),((),Type=Bone,Name=\"dfrm_arm_spline_06_l\"),((),Type=B ...(truncated)") {template="DISPATCH_RigVMDispatch_ArrayMake(in Values,out Array)"}
   > SplineIK = RigVMFunctionReferenceNode(
    Primary Axis=(X=1,Y=0,Z=0)
    Samples Per Segment=15
    Secondary Spline Direction=(X=0,Y=1,Z=0)
    Spline Mode=BSpline
    Stretch=true
    Up Axis=(X=0,Y=1,Z=0)
    debug=false
    Bones=@RerouteNode_16.Array
    Controls=RigVMVariableNode() {variable=arm_l_ctrls}.Value
   )
   > &RigUnit_ParentConstraint(
    AdvancedSettings=(InterpolationType=Shortest,RotationOrderForFilter=XZY)
    Child=(Type=Null,Name="dfrm_arm_spline_01_r_ctrl_null")
    Filter=(TranslationFilter=(bX=false,bY=false,bZ=false),RotationFilter=(bX=True,bY=True,bZ=True),ScaleFilter=(bX=True,bY=True,bZ=True))
    Parents=((Item=(Type=Bone,Name="upperarm_r"),Weight=0.3),(Item=(Type=Bone,Name="spine_05"),Weight=0.7))
    Weight=1
   )
   > &RigUnit_ParentConstraint(
    AdvancedSettings=(InterpolationType=Shortest,RotationOrderForFilter=XZY)
    Child=(Type=Null,Name="dfrm_arm_spline_05_r_ctrl_null")
    Parents=((Item=(Type=Bone,Name="upperarm_r"),Weight=0.2),(Item=(Type=Bone,Name="lowerarm_r"),Weight=0.5),(Item=(Type=Control,Name="dfrm_arm_spline_01_r_ctrl"),Weight=0.3))
    Weight=1
   )
   > &RigUnit_ParentConstraint(
    AdvancedSettings=(InterpolationType=Shortest,RotationOrderForFilter=XZY)
    Child=(Type=Null,Name="dfrm_arm_spline_03_r_ctrl_null")
    Parents=((Item=(Type=Control,Name="dfrm_arm_spline_01_r_ctrl"),Weight=0.7),(Item=(Type=Control,Name="dfrm_arm_spline_05_r_ctrl"),Weight=0.3))
    Weight=1
   )
   > &RigUnit_ParentConstraint(
    AdvancedSettings=(InterpolationType=Shortest,RotationOrderForFilter=XZY)
    Child=(Type=Null,Name="dfrm_arm_spline_09_r_ctrl_null")
    Parents=((Item=(Type=Bone,Name="hand_r"),Weight=0.7),(Item=(Type=Bone,Name="lowerarm_r"),Weight=0.3))
    Weight=1
   )
   > &RigUnit_ParentConstraint(
    AdvancedSettings=(InterpolationType=Shortest,RotationOrderForFilter=XZY)
    Child=(Type=Null,Name="dfrm_arm_spline_07_r_ctrl_null")
    Parents=((Item=(Type=Control,Name="dfrm_arm_spline_05_r_ctrl"),Weight=0.6),(Item=(Type=Control,Name="dfrm_arm_spline_09_r_ctrl"),Weight=0.4))
    Weight=1
   )
   > &RigUnit_ModifyTransforms(
    ItemToModify=((Item=(Type=Control,Name="dfrm_arm_spline_01_r_ctrl"),Transform=(Rotation=(X=0,Y=0,Z=0,W=1),Translation=(X=-4,Y=1.5,Z=2),Scale3D=(X=1,Y=1,Z=1))))
    Weight=RigVMVariableNode() {variable=UseBackpack_PuzzleShed}.Value
   )
   > &RigUnit_ModifyTransforms(
    ItemToModify=((Item=(Type=Control,Name="dfrm_arm_spline_03_r_ctrl"),Transform=(Rotation=(X=0,Y=0,Z=0,W=1),Translation=(X=0,Y=0,Z=-5),Scale3D=(X=1,Y=1,Z=1))))
    Weight=RigVMVariableNode() {variable=spr_upper_l_DBNO}.Value
   )
   > &RigUnit_ModifyTransforms(
    ItemToModify="((Item=(Type=Control,Name=\"dfrm_arm_spline_03_r_ctrl\"),Transform=(Rotation=(X=0,Y=0,Z=0,W=1),Translation=(X=-10,Y=0,Z=0),Scale3D=(X=1,Y=1,Z=1))),(Item=(Type=Control,Name=\"dfrm_arm_spline_05_r_ctrl\"),Transform=(Rotation=(X=0,Y=0,Z=0,W=1),Translation=(X=-15, ...(truncated)"
    Weight=RigVMVariableNode() {variable=spr_upper_r_up_alpha}.Value
   )
   RerouteNode_2_1 = RigVMDispatchNode(Values="((Type=Bone,Name=\"dfrm_clavicle_r\"),((),Type=Bone,Name=\"dfrm_arm_spline_02_r\"),((),Type=Bone,Name=\"dfrm_arm_spline_03_r\"),((),Type=Bone,Name=\"dfrm_arm_spline_04_r\"),((),Type=Bone,Name=\"dfrm_lowerarm_r\"),((),Type=Bone,Name=\"dfrm_arm_spline_06_r\"),((),Type=B ...(truncated)") {template="DISPATCH_RigVMDispatch_ArrayMake(in Values,out Array)"}
   > SplineIK_1 = RigVMFunctionReferenceNode(
    Primary Axis=(X=-1,Y=0,Z=0)
    Samples Per Segment=15
    Secondary Spline Direction=(X=0,Y=1,Z=0)
    Spline Mode=BSpline
    Stretch=true
    Up Axis=(X=0,Y=1,Z=0)
    debug=false
    Bones=@RerouteNode_2_1.Array
    Controls=RigVMVariableNode() {variable=arm_r_ctrls}.Value
   )
   n1 = RigVMVariableNode() {variable=spr_thigh_l_up_alpha}
   > &RigUnit_RotationConstraintLocalSpaceOffset(
    AdvancedSettings=(RotationOrderForFilter=XZY)
    Child=(Type=Bone,Name="def_groin_l")
    Filter=(bX=True,bY=True,bZ=True)
    Parents=((Item=(Type=Bone,Name="pelvis"),Weight=0.5),(Item=(Type=Bone,Name="thigh_twist_01_l"),Weight=0.5))
    Weight=@n1.Value
   )
   > &RigUnit_RotationConstraintLocalSpaceOffset(
    AdvancedSettings=(RotationOrderForFilter=XZY)
    Child=(Type=Bone,Name="def_hip_l")
    Filter=(bX=True,bY=True,bZ=True)
    Parents=((Item=(Type=Bone,Name="pelvis"),Weight=0.5),(Item=(Type=Bone,Name="thigh_twist_01_l"),Weight=0.5))
    Weight=@n1.Value
   )
   > &RigUnit_ModifyTransforms(
    ItemToModify=((Item=(Type=Bone,Name="def_groin_l"),Transform=(Rotation=(X=0,Y=0,Z=0,W=1),Translation=(X=10,Y=20,Z=0),Scale3D=(X=1,Y=1,Z=1))))
    Weight=@n1.Value
   )
   n2 = RigVMVariableNode() {variable=arms}
   > &RigUnit_ModifyTransforms(
    ItemToModify=((Item=(Type=Bone,Name="dfrm_arm_spline_07_r"),Transform=(Rotation=(X=0,Y=0,Z=0,W=1),Translation=(X=0,Y=0,Z=0),Scale3D=(X=0.3,Y=0.3,Z=0.3))))
    Weight=@n2.Value
   )
   > &RigUnit_ModifyTransforms(
    ItemToModify=((Item=(Type=Bone,Name="dfrm_arm_spline_06_r"),Transform=(Rotation=(X=0,Y=0,Z=0,W=1),Translation=(X=0,Y=0,Z=0),Scale3D=(X=0.5,Y=0.5,Z=0.5))))
    Weight=RigVMVariableNode() {variable=arms_1}.Value
   )
   > &RigUnit_ParentConstraint(
    AdvancedSettings=(InterpolationType=Shortest,RotationOrderForFilter=XZY)
    Child=(Type=Bone,Name="dfrm_arm_spline_mstr_r")
    Filter=(TranslationFilter=(bX=true,bY=true,bZ=true),RotationFilter=(bX=true,bY=true,bZ=true),ScaleFilter=(bX=false,bY=false,bZ=false))
    Parents=((Item=(Type=Bone,Name="dfrm_arm_spline_mstr_r"),Weight=0.8),(Item=(Type=Bone,Name="upperarm_r"),Weight=0.2))
    Weight=@n2.Value
   )
   > &RigUnit_ParentConstraint(
    AdvancedSettings=(InterpolationType=Shortest,RotationOrderForFilter=XZY)
    Child=(Type=Bone,Name="dfrm_lowerarm_r")
    Filter=(TranslationFilter=(bX=true,bY=true,bZ=true),RotationFilter=(bX=true,bY=true,bZ=true),ScaleFilter=(bX=false,bY=false,bZ=false))
    Parents=((Item=(Type=Bone,Name="dfrm_lowerarm_r"),Weight=0.35),(Item=(Type=Bone,Name="lowerarm_r"),Weight=0.65))
    Weight=@n2.Value
   )
   > &RigUnit_ParentConstraint(
    AdvancedSettings=(InterpolationType=Shortest,RotationOrderForFilter=XZY)
    Child=(Type=Bone,Name="dfrm_arm_spline_06_r")
    Filter=(TranslationFilter=(bX=true,bY=true,bZ=true),RotationFilter=(bX=true,bY=true,bZ=true),ScaleFilter=(bX=false,bY=false,bZ=false))
    Parents=((Item=(Type=Bone,Name="dfrm_arm_spline_06_r"),Weight=0),(Item=(Type=Bone,Name="lowerarm_r"),Weight=1))
    Weight=@n2.Value
   )
   n3 = RigVMVariableNode() {variable=arms_0}
   > &RigUnit_ModifyTransforms(
    ItemToModify=((Item=(Type=Bone,Name="dfrm_arm_spline_07_l"),Transform=(Rotation=(X=0,Y=0,Z=0,W=1),Translation=(X=0,Y=0,Z=0),Scale3D=(X=0.3,Y=0.3,Z=0.3))))
    Weight=@n3.Value
   )
   > &RigUnit_ModifyTransforms(
    ItemToModify=((Item=(Type=Bone,Name="dfrm_arm_spline_06_l"),Transform=(Rotation=(X=0,Y=0,Z=0,W=1),Translation=(X=0,Y=0,Z=0),Scale3D=(X=0.5,Y=0.5,Z=0.5))))
    Weight=RigVMVariableNode() {variable=arms_1}.Value
   )
   > &RigUnit_ParentConstraint(
    AdvancedSettings=(InterpolationType=Shortest,RotationOrderForFilter=XZY)
    Child=(Type=Bone,Name="dfrm_arm_spline_mstr_l")
    Filter=(TranslationFilter=(bX=true,bY=true,bZ=true),RotationFilter=(bX=true,bY=true,bZ=true),ScaleFilter=(bX=false,bY=false,bZ=false))
    Parents=((Item=(Type=Bone,Name="dfrm_arm_spline_mstr_l"),Weight=0.8),(Item=(Type=Bone,Name="upperarm_l"),Weight=0.2))
    Weight=@n3.Value
   )
   > &RigUnit_ParentConstraint(
    AdvancedSettings=(InterpolationType=Shortest,RotationOrderForFilter=XZY)
    Child=(Type=Bone,Name="dfrm_lowerarm_l")
    Filter=(TranslationFilter=(bX=true,bY=true,bZ=true),RotationFilter=(bX=true,bY=true,bZ=true),ScaleFilter=(bX=false,bY=false,bZ=false))
    Parents=((Item=(Type=Bone,Name="dfrm_lowerarm_l"),Weight=0.35),(Item=(Type=Bone,Name="lowerarm_l"),Weight=0.65))
    Weight=@n3.Value
   )
   > &RigUnit_ParentConstraint(
    AdvancedSettings=(InterpolationType=Shortest,RotationOrderForFilter=XZY)
    Child=(Type=Bone,Name="dfrm_arm_spline_06_l")
    Filter=(TranslationFilter=(bX=true,bY=true,bZ=true),RotationFilter=(bX=true,bY=true,bZ=true),ScaleFilter=(bX=false,bY=false,bZ=false))
    Parents=((Item=(Type=Bone,Name="dfrm_arm_spline_06_l"),Weight=0),(Item=(Type=Bone,Name="lowerarm_l"),Weight=1))
    Weight=@n3.Value
   )
   &RigUnit_RotationConstraintLocalSpaceOffset(
    AdvancedSettings=(RotationOrderForFilter=XZY)
    Child=(Type=Null,Name="dfrm_leg_spline_01_l_ctrl_null")
    Filter=(bX=True,bY=True,bZ=True)
    Parents=((Item=(Type=Bone,Name="thigh_l"),Weight=1))
    Weight=1
    ExecutePin=@SequenceAggregate_1.C
   )
   > &RigUnit_ParentConstraint(
    AdvancedSettings=(RotationOrderForFilter=XZY)
    Child=(Type=Null,Name="dfrm_leg_spline_03_l_ctrl_null")
    Parents=((Item=(Type=Bone,Name="thigh_l"),Weight=1),(Item=(Type=Bone,Name="calf_l"),Weight=0))
    Weight=1
   )
   > &RigUnit_ParentConstraint(
    AdvancedSettings=(RotationOrderForFilter=XZY)
    Child=(Type=Null,Name="dfrm_leg_spline_05_l_ctrl_null")
    Parents=((Item=(Type=Bone,Name="thigh_l"),Weight=0.5),(Item=(Type=Bone,Name="calf_l"),Weight=0.5))
    Weight=1
   )
   > &RigUnit_ParentConstraint(
    AdvancedSettings=(RotationOrderForFilter=XZY)
    Child=(Type=Null,Name="dfrm_leg_spline_07_l_ctrl_null")
    Parents=((Item=(Type=Bone,Name="calf_l"),Weight=1),(Item=(Type=Bone,Name="foot_l"),Weight=0))
    Weight=1
   )
   RerouteNode_4_1 = RigVMDispatchNode(Values="((Type=Bone,Name=\"dfrm_leg_spline_01_l\"),((),Type=Bone,Name=\"dfrm_leg_spline_02_l\"),((),Type=Bone,Name=\"dfrm_leg_spline_03_l\"),((),Type=Bone,Name=\"dfrm_leg_spline_04_l\"),((),Type=Bone,Name=\"dfrm_leg_spline_05_l\"),((),Type=Bone,Name=\"dfrm_leg_spline_06_l\"), ...(truncated)") {template="DISPATCH_RigVMDispatch_ArrayMake(in Values,out Array)"}
   > SplineIK_1_1 = RigVMFunctionReferenceNode(
    Primary Axis=(X=1,Y=0,Z=0)
    Samples Per Segment=2
    Secondary Spline Direction=(X=0,Y=0,Z=0)
    Spline Mode=BSpline
    Stretch=true
    Up Axis=(X=0,Y=0,Z=0)
    debug=false
    Bones=@RerouteNode_4_1.Array
    Controls=RigVMVariableNode() {variable=leg_l_ctrls}.Value
   )
   > &RigUnit_RotationConstraintLocalSpaceOffset(
    AdvancedSettings=(RotationOrderForFilter=XZY)
    Child=(Type=Null,Name="dfrm_leg_spline_01_r_ctrl_null")
    Filter=(bX=True,bY=True,bZ=True)
    Parents=((Item=(Type=Bone,Name="thigh_r"),Weight=1))
    Weight=1
   )
   > &RigUnit_ParentConstraint(
    AdvancedSettings=(RotationOrderForFilter=XZY)
    Child=(Type=Null,Name="dfrm_leg_spline_03_r_ctrl_null")
    Parents=((Item=(Type=Bone,Name="thigh_r"),Weight=1),(Item=(Type=Bone,Name="calf_r"),Weight=0))
    Weight=1
   )
   > &RigUnit_ParentConstraint(
    AdvancedSettings=(RotationOrderForFilter=XZY)
    Child=(Type=Null,Name="dfrm_leg_spline_05_r_ctrl_null")
    Parents=((Item=(Type=Bone,Name="thigh_r"),Weight=0.5),(Item=(Type=Bone,Name="calf_r"),Weight=0.5))
    Weight=1
   )
   > &RigUnit_ParentConstraint(
    AdvancedSettings=(RotationOrderForFilter=XZY)
    Child=(Type=Null,Name="dfrm_leg_spline_07_r_ctrl_null")
    Parents=((Item=(Type=Bone,Name="calf_r"),Weight=1),(Item=(Type=Bone,Name="foot_r"),Weight=0))
    Weight=1
   )
   RerouteNode_1_3 = RigVMDispatchNode(Values="((Type=Bone,Name=\"dfrm_leg_spline_01_r\"),((),Type=Bone,Name=\"dfrm_leg_spline_02_r\"),((),Type=Bone,Name=\"dfrm_leg_spline_03_r\"),((),Type=Bone,Name=\"dfrm_leg_spline_04_r\"),((),Type=Bone,Name=\"dfrm_leg_spline_05_r\"),((),Type=Bone,Name=\"dfrm_leg_spline_06_r\"), ...(truncated)") {template="DISPATCH_RigVMDispatch_ArrayMake(in Values,out Array)"}
   > SplineIK_1_1_2 = RigVMFunctionReferenceNode(
    Primary Axis=(X=-1,Y=0,Z=0)
    Samples Per Segment=2
    Secondary Spline Direction=(X=0,Y=0,Z=0)
    Spline Mode=BSpline
    Stretch=true
    Up Axis=(X=0,Y=0,Z=0)
    debug=false
    Bones=@RerouteNode_1_3.Array
    Controls=RigVMVariableNode() {variable=leg_r_ctrls}.Value
   )
   n4 = RigVMVariableNode() {variable=spr_thigh_r_up_alpha}
   > &RigUnit_RotationConstraintLocalSpaceOffset(
    AdvancedSettings=(RotationOrderForFilter=XZY)
    Child=(Type=Bone,Name="def_groin_r")
    Filter=(bX=True,bY=True,bZ=True)
    Parents=((Item=(Type=Bone,Name="pelvis"),Weight=0.5),(Item=(Type=Bone,Name="thigh_twist_01_r"),Weight=0.5))
    Weight=@n4.Value
   )
   > &RigUnit_RotationConstraintLocalSpaceOffset(
    AdvancedSettings=(RotationOrderForFilter=XZY)
    Child=(Type=Bone,Name="def_hip_r")
    Filter=(bX=True,bY=True,bZ=True)
    Parents=((Item=(Type=Bone,Name="pelvis"),Weight=0.5),(Item=(Type=Bone,Name="thigh_twist_01_r"),Weight=0.5))
    Weight=@n4.Value
   )
   > &RigUnit_ModifyTransforms(
    ItemToModify=((Item=(Type=Bone,Name="def_groin_r"),Transform=(Rotation=(X=0,Y=0,Z=0,W=1),Translation=(X=-10,Y=-20,Z=0),Scale3D=(X=1,Y=1,Z=1))))
    Weight=@n4.Value
   )
   SequenceAggregate_3 = RigVMAggregateNode(ExecuteContext=@SequenceAggregate_1.D)
   n5 = RigVMVariableNode() {variable=spr_upper_l_front_alpha}
   n6 = RigVMVariableNode() {variable=arm_l_ctrls}
   /Script/ControlRig.RigUnit_HierarchySetPoseItemArray(
    ElementType=All
    ExecutePin=@SequenceAggregate_3.A
    Weight=&RigVMFunction_MathDoubleSub(A=1, B=@n5.Value).Result
    Pose=RigVMVariableNode() {variable=arm_l_bind_pose}.Value
    ItemsToSet=@n6.Value
   ) {is=struct RigVMUnitNode}
   n7 = RigVMVariableNode() {variable=spr_upper_r_front_alpha}
   n8 = RigVMVariableNode() {variable=arm_r_ctrls}
   &RigUnit_HierarchySetPoseItemArray(
    ElementType=All
    Weight=&RigVMFunction_MathDoubleSub(A=1, B=@n7.Value).Result
    ItemsToSet=@n8.Value
    Pose=RigVMVariableNode() {variable=arm_r_bind_pose}.Value
    ExecutePin=@SequenceAggregate_3.C
   )
   &RigUnit_HierarchySetPoseItemArray(
    ElementType=All
    Pose=RigVMVariableNode() {variable=upperarm_r_front_pose}.Value
    ItemsToSet=@n8.Value
    Weight=@n7.Value
    ExecutePin=@SequenceAggregate_3.D
   )
   &RigUnit_HierarchySetPoseItemArray(
    ElementType=All
    Pose=RigVMVariableNode() {variable=upperarm_l_front_pose}.Value
    ItemsToSet=@n6.Value
    Weight=@n5.Value
    ExecutePin=@SequenceAggregate_3.B
   )
   SequenceAggregate_2 = RigVMAggregateNode(ExecuteContext=@SequenceAggregate_1.F)
   n9 = RigVMVariableNode() {variable=spr_thigh_l_up_alpha}
   n10 = RigVMVariableNode() {variable=thigh_l_ctrls}
   &RigUnit_HierarchySetPoseItemArray(
    ElementType=All
    ExecutePin=@SequenceAggregate_2.A
    Weight=&RigVMFunction_MathDoubleSub(A=1, B=@n9.Value).Result
    ItemsToSet=@n10.Value
    Pose=RigVMVariableNode() {variable=thigh_l_bind_pose}.Value
   )
   &RigUnit_HierarchySetPoseItemArray(
    ElementType=All
    ItemsToSet=@n10.Value
    Pose=RigVMVariableNode() {variable=thigh_l_up_pose}.Value
    Weight=@n9.Value
    ExecutePin=@SequenceAggregate_2.B
   )
   n11 = RigVMVariableNode() {variable=spr_thigh_r_up_alpha}
   n12 = RigVMVariableNode() {variable=thigh_r_ctrls}
   &RigUnit_HierarchySetPoseItemArray(
    ElementType=All
    Weight=&RigVMFunction_MathDoubleSub(A=1, B=@n11.Value).Result
    ItemsToSet=@n12.Value
    Pose=RigVMVariableNode() {variable=thigh_r_bind_pose}.Value
    ExecutePin=@SequenceAggregate_2.C
   )
   &RigUnit_HierarchySetPoseItemArray(
    ElementType=All
    ItemsToSet=@n12.Value
    Weight=@n11.Value
    Pose=RigVMVariableNode() {variable=thigh_r_up_pose}.Value
    ExecutePin=@SequenceAggregate_2.D
   )
   SequenceAggregate_2_1 = RigVMAggregateNode(ExecuteContext=@SequenceAggregate_1.G)
   n13 = RigVMVariableNode() {variable=spr_knee_l_rot_alpha}
   n14 = RigVMVariableNode() {variable=calf_l_ctrls}
   &RigUnit_HierarchySetPoseItemArray(
    ElementType=All
    ExecutePin=@SequenceAggregate_2_1.A
    Weight=&RigVMFunction_MathDoubleSub(A=1, B=@n13.Value).Result
    ItemsToSet=@n14.Value
    Pose=RigVMVariableNode() {variable=calf_l_bind_pose}.Value
   )
   n15 = RigVMVariableNode() {variable=spr_knee_r_rot_alpha}
   n16 = RigVMVariableNode() {variable=calf_r_ctrls}
   &RigUnit_HierarchySetPoseItemArray(
    ElementType=All
    Weight=&RigVMFunction_MathDoubleSub(A=1, B=@n15.Value).Result
    ItemsToSet=@n16.Value
    Pose=RigVMVariableNode() {variable=calf_r_bind_pose}.Value
    ExecutePin=@SequenceAggregate_2_1.C
   )
   &RigUnit_HierarchySetPoseItemArray(
    ElementType=All
    Pose=RigVMVariableNode() {variable=knee_l_rot_pose}.Value
    Weight=@n13.Value
    ItemsToSet=@n14.Value
    ExecutePin=@SequenceAggregate_2_1.B
   )
   &RigUnit_HierarchySetPoseItemArray(
    ElementType=All
    Pose=RigVMVariableNode() {variable=knee_r_rot_pose}.Value
    Weight=@n15.Value
    ItemsToSet=@n16.Value
    ExecutePin=@SequenceAggregate_2_1.D
   )
   SequenceAggregate = RigVMAggregateNode(ExecuteContext=@SequenceAggregate_1.H)
   DISPATCH_RigVMDispatch_ArrayIterator = RigVMDispatchNode(
    Array=&RigUnit_ItemArray(Items=((Type=Bone,Name="neck_01"),(Type=Bone,Name="neck_02"),(Type=Bone,Name="head"))).Items
    ExecuteContext=@SequenceAggregate.A
   ) {template="DISPATCH_RigVMDispatch_ArrayIterator(in Array,out Element,out Index,out Count,out Ratio)"}
   RigUnit_GetTransform = &RigUnit_GetTransform(Space=LocalSpace, bInitial=false, Item=@DISPATCH_RigVMDispatch_ArrayIterator.Element)
   RigVMFunction_MathQuaternionToEuler = &RigVMFunction_MathQuaternionToEuler(RotationOrder=XYZ, Value=@RigUnit_GetTransform.Transform.Rotation)
   Multiply = &RigVMFunction_MathDoubleMul(B=1, A=@RigVMFunction_MathQuaternionToEuler.Result.Z)
   RigVMFunction_MathQuaternionFromEuler = &RigVMFunction_MathQuaternionFromEuler(
    Euler=(X=0,Y=0,Z=0)
    RotationOrder=ZXY
    Euler.Z=@Multiply.Result
    Euler.Y=&RigVMFunction_MathDoubleMul(B=1, A=@RigVMFunction_MathQuaternionToEuler.Result.Y).Result
    Euler.X=&RigVMFunction_MathDoubleMul(B=0.6, A=@RigVMFunction_MathQuaternionToEuler.Result.X).Result
   )
   > RigUnit_SetTransform = &RigUnit_SetTransform(
    Space=LocalSpace
    Value=(Rotation=(X=-0.0029,Y=0.7071,Z=0.0029,W=-0.7071),Translation=(X=0,Y=-1.6873,Z=152.0342),Scale3D=(X=1,Y=1,Z=1))
    Weight=1
    bInitial=False
    Item=@DISPATCH_RigVMDispatch_ArrayIterator.Element
    Value.Rotation=@RigVMFunction_MathQuaternionFromEuler.Result
    Value.Translation=@RigUnit_GetTransform.Transform.Translation
    Value.Scale3D=@RigUnit_GetTransform.Transform.Scale3D
   )
   n17 = &RigUnit_SphericalPoseReader(
    ActiveRegionScaleFactors=(PositiveWidth=2,NegativeWidth=2,PositiveHeight=2,NegativeHeight=2)
    ActiveRegionSize=0.1
    Debug=(bDrawDebug=false,DebugScale=25,DebugSegments=20,DebugThickness=0.25)
    DriverAxis=(X=0,Y=1,Z=0)
    DriverItem=(Type=Bone,Name="head")
    FalloffRegionScaleFactors=(PositiveWidth=1.35,NegativeWidth=1.25,PositiveHeight=1.35,NegativeHeight=1.7)
    FalloffSize=0.4
    FlipHeightScaling=false
    FlipWidthScaling=true
    OptionalParentItem=(Type=Bone,Name="pelvis")
    RotationOffset=(X=90,Y=0,Z=85)
    ExecutePin=@SequenceAggregate.B
   )
   > &RigUnit_ModifyTransforms(
    ItemToModify=((Item=(Type=Bone,Name="head"),Transform=(Rotation=(X=0,Y=0,Z=-0.1478,W=0.989),Translation=(X=0,Y=0,Z=0),Scale3D=(X=1,Y=1,Z=1))))
    Weight=@n17.OutputParam
   )
   n18 = &RigUnit_SphericalPoseReader(
    ActiveRegionScaleFactors=(PositiveWidth=3.5,NegativeWidth=0.5,PositiveHeight=3,NegativeHeight=3)
    ActiveRegionSize=0.1
    Debug=(bDrawDebug=false,DebugScale=25,DebugSegments=20,DebugThickness=0.25)
    DriverAxis=(X=-1,Y=0,Z=0)
    DriverItem=(Type=Bone,Name="thigh_twist_01_l")
    FalloffRegionScaleFactors=(PositiveWidth=1,NegativeWidth=0.8,PositiveHeight=1,NegativeHeight=1)
    FalloffSize=0.55
    OptionalParentItem=(Type=Bone,Name="pelvis")
    RotationOffset=(X=90,Y=0,Z=-15)
    ExecutePin=@SequenceAggregate_1.A
   )
   > RigVMVariableNode(Value=@n18.OutputParam) {variable=spr_thigh_l_up_alpha}
   > n19 = &RigUnit_SphericalPoseReader(
    ActiveRegionScaleFactors=(PositiveWidth=3.5,NegativeWidth=0.5,PositiveHeight=3,NegativeHeight=3)
    ActiveRegionSize=0.1
    Debug=(bDrawDebug=false,DebugScale=25,DebugSegments=20,DebugThickness=0.25)
    DriverAxis=(X=-1,Y=0,Z=0)
    DriverItem=(Type=Bone,Name="calf_l")
    FalloffRegionScaleFactors=(PositiveWidth=1,NegativeWidth=0.9,PositiveHeight=1,NegativeHeight=1)
    FalloffSize=0.45
    OptionalParentItem=(Type=Bone,Name="thigh_l")
    RotationOffset=(X=-90,Y=0,Z=0)
   )
   > RigVMVariableNode(Value=@n19.OutputParam) {variable=spr_knee_l_rot_alpha}
   > n20 = &RigUnit_SphericalPoseReader(
    ActiveRegionScaleFactors=(PositiveWidth=3.5,NegativeWidth=0.5,PositiveHeight=3,NegativeHeight=3)
    ActiveRegionSize=0.1
    Debug=(bDrawDebug=false,DebugScale=25,DebugSegments=20,DebugThickness=0.25)
    DriverAxis=(X=1,Y=0,Z=0)
    DriverItem=(Type=Bone,Name="thigh_twist_01_r")
    FalloffRegionScaleFactors=(PositiveWidth=1,NegativeWidth=1,PositiveHeight=1,NegativeHeight=1)
    FalloffSize=0.55
    FlipHeightScaling=false
    FlipWidthScaling=true
    OptionalParentItem=(Type=Bone,Name="pelvis")
    RotationOffset=(X=-90,Y=0,Z=-15)
   )
   > RigVMVariableNode(Value=@n20.OutputParam) {variable=spr_thigh_r_up_alpha}
   > n21 = &RigUnit_SphericalPoseReader(
    ActiveRegionScaleFactors=(PositiveWidth=3.5,NegativeWidth=0.5,PositiveHeight=3,NegativeHeight=3)
    ActiveRegionSize=0.1
    Debug=(bDrawDebug=false,DebugScale=25,DebugSegments=20,DebugThickness=0.25)
    DriverAxis=(X=1,Y=0,Z=0)
    DriverItem=(Type=Bone,Name="calf_r")
    FalloffRegionScaleFactors=(PositiveWidth=1,NegativeWidth=1,PositiveHeight=1,NegativeHeight=1)
    FalloffSize=0.45
    FlipHeightScaling=false
    FlipWidthScaling=true
    OptionalParentItem=(Type=Bone,Name="thigh_r")
    RotationOffset=(X=90,Y=0,Z=0)
   )
   > RigVMVariableNode(Value=@n21.OutputParam) {variable=spr_knee_r_rot_alpha}
   > n22 = &RigUnit_SphericalPoseReader(
    ActiveRegionScaleFactors=(PositiveWidth=1.75,NegativeWidth=3,PositiveHeight=1.5,NegativeHeight=1.2)
    ActiveRegionSize=0.1
    Debug=(bDrawDebug=false,bDraw2D=false,bDrawLocalAxes=false,DebugScale=25,DebugSegments=20,DebugThickness=0.25)
    DriverAxis=(X=1,Y=0,Z=0)
    DriverItem=(Type=Bone,Name="upperarm_l")
    FalloffRegionScaleFactors=(PositiveWidth=1,NegativeWidth=5,PositiveHeight=1,NegativeHeight=1)
    FalloffSize=0.35
    OptionalParentItem=(Type=Bone,Name="spine_04")
    RotationOffset=(X=100,Y=-45,Z=0)
   )
   > RigVMVariableNode(Value=@n22.OutputParam) {variable=spr_upper_l_front_alpha}
   > n23 = &RigUnit_SphericalPoseReader(
    ActiveRegionScaleFactors=(PositiveWidth=2,NegativeWidth=3,PositiveHeight=1,NegativeHeight=1)
    ActiveRegionSize=0.1
    Debug=(bDrawDebug=false,DebugScale=25,DebugSegments=20,DebugThickness=0.25)
    DriverAxis=(X=-1,Y=0,Z=0)
    DriverItem=(Type=Bone,Name="upperarm_r")
    FalloffRegionScaleFactors=(PositiveWidth=1.5,NegativeWidth=5,PositiveHeight=1,NegativeHeight=0.8)
    FalloffSize=0.35
    FlipHeightScaling=false
    FlipWidthScaling=true
    OptionalParentItem=(Type=Bone,Name="spine_04")
    RotationOffset=(X=-60,Y=-45,Z=0)
   )
   > RigVMVariableNode(Value=@n23.OutputParam) {variable=spr_upper_r_front_alpha}
   > n24 = &RigUnit_SphericalPoseReader(
    ActiveRegionScaleFactors=(PositiveWidth=2,NegativeWidth=2,PositiveHeight=2,NegativeHeight=2)
    ActiveRegionSize=0.2
    Debug=(bDrawDebug=false,DebugScale=25,DebugSegments=20,DebugThickness=0.25)
    DriverAxis=(X=0,Y=1,Z=0)
    DriverItem=(Type=Bone,Name="spine_04")
    FalloffRegionScaleFactors=(PositiveWidth=1,NegativeWidth=1,PositiveHeight=1,NegativeHeight=1)
    FalloffSize=0.8
    FlipHeightScaling=false
    FlipWidthScaling=false
    OptionalParentItem=(Type=Bone,Name="spine_02")
    RotationOffset=(X=-95,Y=0,Z=0)
   )
   > RigVMVariableNode(Value=@n24.OutputParam) {variable=chest_twist_alpha}
   > n25 = &RigUnit_SphericalPoseReader(
    ActiveRegionScaleFactors=(PositiveWidth=1.75,NegativeWidth=1.5,PositiveHeight=1.5,NegativeHeight=1.2)
    ActiveRegionSize=0.1
    Debug=(bDrawDebug=false,bDraw2D=false,bDrawLocalAxes=false,DebugScale=25,DebugSegments=20,DebugThickness=0.25)
    DriverAxis=(X=1,Y=0,Z=0)
    DriverItem=(Type=Bone,Name="upperarm_l")
    FalloffRegionScaleFactors=(PositiveWidth=1,NegativeWidth=1.5,PositiveHeight=1,NegativeHeight=1)
    FalloffSize=0.35
    OptionalParentItem=(Type=Bone,Name="spine_04")
    RotationOffset=(X=0,Y=0,Z=0)
   )
   > RigVMVariableNode(Value=@n25.OutputParam) {variable=spr_upper_l_up_alpha}
   > n26 = &RigUnit_SphericalPoseReader(
    ActiveRegionScaleFactors=(PositiveWidth=1.75,NegativeWidth=1.5,PositiveHeight=1.5,NegativeHeight=1.2)
    ActiveRegionSize=0.1
    Debug=(bDrawDebug=false,bDraw2D=false,bDrawLocalAxes=false,DebugScale=25,DebugSegments=20,DebugThickness=0.25)
    DriverAxis=(X=-1,Y=0,Z=0)
    DriverItem=(Type=Bone,Name="upperarm_r")
    FalloffRegionScaleFactors=(PositiveWidth=1,NegativeWidth=1.5,PositiveHeight=1,NegativeHeight=1)
    FalloffSize=0.35
    FlipHeightScaling=true
    FlipWidthScaling=true
    OptionalParentItem=(Type=Bone,Name="spine_04")
    RotationOffset=(X=180,Y=0,Z=0)
   )
   > RigVMVariableNode(Value=@n26.OutputParam) {variable=spr_upper_r_up_alpha}
   RigVMVariableNode() {variable=arms_1}
   RigVMVariableNode() {variable=arms_1}
  ]}
 RigVMModel Rig Graph {nodes [
   n1 = &RigUnit_PrepareForExecution()
   SequenceAggregate_1 = RigVMAggregateNode(ExecuteContext=@n1.ExecutePin)
   SequenceAggregate = RigVMAggregateNode(ExecuteContext=@SequenceAggregate_1.A)
   RerouteNode_8 = RigVMDispatchNode(Values=((Type=Control,Name="dfrm_arm_spline_01_r_ctrl"),((),Type=Control,Name="dfrm_arm_spline_03_r_ctrl"),((),Type=Control,Name="dfrm_arm_spline_05_r_ctrl"),((),Type=Control,Name="dfrm_arm_spline_07_r_ctrl"),((),Type=Control,Name="dfrm_arm_spline_09_r_ctrl"))) {template="DISPATCH_RigVMDispatch_ArrayMake(in Values,out Array)"}
   RigVMVariableNode(Value=@RerouteNode_8.Array, ExecuteContext=@SequenceAggregate.B) {variable=arm_r_ctrls}
   > RigVMVariableNode(Value=/Script/ControlRig.RigUnit_HierarchyGetPoseItemArray(ElementType=All, ItemsToGet=@RerouteNode_8.Array) {is=struct RigVMUnitNode}.Pose) {variable=arm_r_bind_pose}
   RerouteNode_9 = RigVMDispatchNode(Values=((Type=Control,Name="dfrm_leg_spline_01_l_ctrl"),((),Type=Control,Name="dfrm_leg_spline_03_l_ctrl"),((),Type=Control,Name="dfrm_leg_spline_05_l_ctrl"),((),Type=Control,Name="dfrm_leg_spline_07_l_ctrl"),((),Type=Control,Name="dfrm_leg_spline_09_l_ctrl"))) {template="DISPATCH_RigVMDispatch_ArrayMake(in Values,out Array)"}
   RigVMVariableNode(Value=@RerouteNode_9.Array, ExecuteContext=@SequenceAggregate.C) {variable=leg_l_ctrls}
   > RigVMVariableNode(Value=&RigUnit_HierarchyGetPoseItemArray(ElementType=All, ItemsToGet=@RerouteNode_9.Array).Pose) {variable=leg_l_bind_pose}
   RerouteNode_12 = RigVMDispatchNode(Values=((Type=Control,Name="dfrm_leg_spline_01_l_ctrl"),((),Type=Control,Name="dfrm_leg_spline_03_l_ctrl"))) {template="DISPATCH_RigVMDispatch_ArrayMake(in Values,out Array)"}
   > RigVMVariableNode(Value=@RerouteNode_12.Array) {variable=thigh_l_ctrls}
   > RigVMVariableNode(Value=&RigUnit_HierarchyGetPoseItemArray(ElementType=All, ItemsToGet=@RerouteNode_12.Array).Pose) {variable=thigh_l_bind_pose}
   RerouteNode_14 = RigVMDispatchNode(Values=((Type=Control,Name="dfrm_leg_spline_05_l_ctrl"),((),Type=Control,Name="dfrm_leg_spline_07_l_ctrl"))) {template="DISPATCH_RigVMDispatch_ArrayMake(in Values,out Array)"}
   > RigVMVariableNode(Value=@RerouteNode_14.Array) {variable=calf_l_ctrls}
   > RigVMVariableNode(Value=&RigUnit_HierarchyGetPoseItemArray(ElementType=All, ItemsToGet=@RerouteNode_14.Array).Pose) {variable=calf_l_bind_pose}
   RerouteNode_10 = RigVMDispatchNode(Values=((Type=Control,Name="dfrm_leg_spline_01_r_ctrl"),((),Type=Control,Name="dfrm_leg_spline_03_r_ctrl"),((),Type=Control,Name="dfrm_leg_spline_05_r_ctrl"),((),Type=Control,Name="dfrm_leg_spline_07_r_ctrl"),((),Type=Control,Name="dfrm_leg_spline_09_r_ctrl"))) {template="DISPATCH_RigVMDispatch_ArrayMake(in Values,out Array)"}
   RigVMVariableNode(Value=@RerouteNode_10.Array, ExecuteContext=@SequenceAggregate.D) {variable=leg_r_ctrls}
   > RigVMVariableNode(Value=&RigUnit_HierarchyGetPoseItemArray(ElementType=All, ItemsToGet=@RerouteNode_10.Array).Pose) {variable=leg_r_bind_pose}
   RerouteNode_13 = RigVMDispatchNode(Values=((Type=Control,Name="dfrm_leg_spline_01_r_ctrl"),((),Type=Control,Name="dfrm_leg_spline_03_r_ctrl"))) {template="DISPATCH_RigVMDispatch_ArrayMake(in Values,out Array)"}
   > RigVMVariableNode(Value=@RerouteNode_13.Array) {variable=thigh_r_ctrls}
   > RigVMVariableNode(Value=&RigUnit_HierarchyGetPoseItemArray(ElementType=All, ItemsToGet=@RerouteNode_13.Array).Pose) {variable=thigh_r_bind_pose}
   RerouteNode_15 = RigVMDispatchNode(Values=((Type=Control,Name="dfrm_leg_spline_05_r_ctrl"),((),Type=Control,Name="dfrm_leg_spline_07_r_ctrl"))) {template="DISPATCH_RigVMDispatch_ArrayMake(in Values,out Array)"}
   > RigVMVariableNode(Value=@RerouteNode_15.Array) {variable=calf_r_ctrls}
   > RigVMVariableNode(Value=&RigUnit_HierarchyGetPoseItemArray(ElementType=All, ItemsToGet=@RerouteNode_15.Array).Pose) {variable=calf_r_bind_pose}
   RerouteNode_6 = RigVMDispatchNode(Values=((Type=Control,Name="dfrm_arm_spline_01_l_ctrl"),((),Type=Control,Name="dfrm_arm_spline_03_l_ctrl"),((),Type=Control,Name="dfrm_arm_spline_05_l_ctrl"),((),Type=Control,Name="dfrm_arm_spline_07_l_ctrl"),((),Type=Control,Name="dfrm_arm_spline_09_l_ctrl"))) {template="DISPATCH_RigVMDispatch_ArrayMake(in Values,out Array)"}
   RigVMVariableNode(Value=@RerouteNode_6.Array, ExecuteContext=@SequenceAggregate.A) {variable=arm_l_ctrls}
   > RigVMVariableNode(Value=&RigUnit_HierarchyGetPoseItemArray(ElementType=All, ItemsToGet=@RerouteNode_6.Array).Pose) {variable=arm_l_bind_pose}
   &RigUnit_ModifyTransforms(
    ItemToModify=((Item=(Type=Control,Name="dfrm_leg_spline_05_l_ctrl"),Transform=(Rotation=(X=0,Y=0,Z=0,W=1),Translation=(X=5,Y=0,Z=0),Scale3D=(X=1,Y=1,Z=1))))
    Weight=1
    ExecutePin=@SequenceAggregate_1.E
   )
   > &RigUnit_ModifyTransforms(
    ItemToModify=((Item=(Type=Control,Name="dfrm_leg_spline_07_l_ctrl"),Transform=(Rotation=(X=0,Y=0,Z=0,W=1),Translation=(X=7,Y=5,Z=0),Scale3D=(X=1,Y=1,Z=1))))
    Weight=1
   )
   > RigVMVariableNode(Value=&RigUnit_HierarchyGetPoseItemArray(ElementType=All, ItemsToGet=RigVMVariableNode() {variable=calf_l_ctrls}.Value).Pose) {variable=knee_l_rot_pose}
   > &RigUnit_ModifyTransforms(
    ItemToModify=((Item=(Type=Control,Name="dfrm_leg_spline_05_r_ctrl"),Transform=(Rotation=(X=0,Y=0,Z=0,W=1),Translation=(X=-3,Y=-5,Z=0),Scale3D=(X=1,Y=1,Z=1))))
    Weight=1
   )
   > &RigUnit_ModifyTransforms(
    ItemToModify=((Item=(Type=Control,Name="dfrm_leg_spline_07_r_ctrl"),Transform=(Rotation=(X=0,Y=0,Z=0,W=1),Translation=(X=0,Y=-2,Z=0),Scale3D=(X=1,Y=1,Z=1))))
    Weight=1
   )
   > RigVMVariableNode(Value=&RigUnit_HierarchyGetPoseItemArray(ElementType=All, ItemsToGet=RigVMVariableNode() {variable=calf_r_ctrls}.Value).Pose) {variable=knee_r_rot_pose}
   &RigUnit_ModifyTransforms(
    ItemToModify=((Item=(Type=Control,Name="dfrm_leg_spline_01_l_ctrl"),Transform=(Rotation=(X=0,Y=0,Z=0,W=1),Translation=(X=-3,Y=0,Z=0),Scale3D=(X=1,Y=1,Z=1))))
    Weight=1
    ExecutePin=@SequenceAggregate_1.B
   )
   > &RigUnit_ModifyTransforms(
    ItemToModify=((Item=(Type=Control,Name="dfrm_leg_spline_03_l_ctrl"),Transform=(Rotation=(X=0,Y=0,Z=0,W=1),Translation=(X=0,Y=0,Z=0),Scale3D=(X=1,Y=1,Z=1))))
    Weight=1
   )
   > RigVMVariableNode(Value=&RigUnit_HierarchyGetPoseItemArray(ElementType=All, ItemsToGet=RigVMVariableNode() {variable=thigh_l_ctrls}.Value).Pose) {variable=thigh_l_up_pose}
   > &RigUnit_ModifyTransforms(
    ItemToModify=((Item=(Type=Control,Name="dfrm_leg_spline_01_r_ctrl"),Transform=(Rotation=(X=0,Y=0,Z=0,W=1),Translation=(X=5,Y=0,Z=0),Scale3D=(X=1,Y=1,Z=1))))
    Weight=1
   )
   > &RigUnit_ModifyTransforms(
    ItemToModify=((Item=(Type=Control,Name="dfrm_leg_spline_03_r_ctrl"),Transform=(Rotation=(X=0,Y=0,Z=0,W=1),Translation=(X=5,Y=5,Z=0),Scale3D=(X=1,Y=1,Z=1))))
    Weight=1
   )
   > RigVMVariableNode(Value=&RigUnit_HierarchyGetPoseItemArray(ElementType=All, ItemsToGet=RigVMVariableNode() {variable=thigh_r_ctrls}.Value).Pose) {variable=thigh_r_up_pose}
   &RigUnit_ModifyTransforms(
    ItemToModify=((Item=(Type=Control,Name="dfrm_arm_spline_03_l_ctrl"),Transform=(Rotation=(X=0,Y=0,Z=0,W=1),Translation=(X=0,Y=0,Z=-8),Scale3D=(X=1,Y=1,Z=1))))
    Weight=1
    ExecutePin=@SequenceAggregate_1.C
   )
   > &RigUnit_ModifyTransforms(
    ItemToModify=((Item=(Type=Control,Name="dfrm_arm_spline_05_l_ctrl"),Transform=(Rotation=(X=0,Y=0,Z=0,W=1),Translation=(X=0,Y=0,Z=-5),Scale3D=(X=1,Y=1,Z=1))))
    Weight=1
   )
   > RigVMVariableNode(Value=&RigUnit_HierarchyGetPoseItemArray(ElementType=All, ItemsToGet=RigVMVariableNode() {variable=arm_l_ctrls}.Value).Pose) {variable=upperarm_l_front_pose}
   &RigUnit_ModifyTransforms(
    ItemToModify=((Item=(Type=Control,Name="dfrm_arm_spline_03_r_ctrl"),Transform=(Rotation=(X=0,Y=0,Z=0,W=1),Translation=(X=0,Y=0,Z=8),Scale3D=(X=1,Y=1,Z=1))))
    Weight=1
    ExecutePin=@SequenceAggregate_1.D
   )
   > &RigUnit_ModifyTransforms(
    ItemToModify=((Item=(Type=Control,Name="dfrm_arm_spline_05_r_ctrl"),Transform=(Rotation=(X=0,Y=0,Z=0,W=1),Translation=(X=0,Y=0,Z=5),Scale3D=(X=1,Y=1,Z=1))))
    Weight=1
   )
   > RigVMVariableNode(Value=&RigUnit_HierarchyGetPoseItemArray(ElementType=All, ItemsToGet=RigVMVariableNode() {variable=arm_r_ctrls}.Value).Pose) {variable=upperarm_r_front_pose}
  ]}
]
hierarchy [
 root {loc=<0 0 0>, rot=<0 0 0 1>}
 pelvis {parent=root, loc=<0 -0.7321 92.4218>, rot=<0.0127 0.707 -0.0127 -0.707>}
 spine_01 {parent=pelvis, loc=<5.6002 -0.5128 0>, rot=<0 0 0.034 -0.9994>}
 spine_02 {parent=spine_01, loc=<8.0756 0.0139 0>, rot=<0 0 0.0238 -0.9997>}
 spine_03 {parent=spine_02, loc=<7.0395 0 0>, rot=<0 0 0.0149 -0.9999>}
 spine_04 {parent=spine_03, loc=<8.1024 0 0>, rot=<0 0 -0.0132 -0.9999>}
 spine_05 {parent=spine_04, loc=<10.0141 0 0>, rot=<0 0 -0.0015 -1>}
 attach_cape {parent=spine_05, loc=<9.1509 -13.3783 0>, rot=<0.0616 -0.7044 -0.0616 -0.7044>}
 neck_01 {parent=spine_05, loc=<10.815 1.9196 0>, rot=<0 0 -0.1864 -0.9825>}
 neck_02 {parent=neck_01, loc=<4.7347 0.3045 0>, rot=<0 0 0.0363 -0.9993>}
 head {parent=neck_02, loc=<5.7632 0.5552 0>, rot=<0 0 0.115 -0.9934>}
 attach_backpack {parent=spine_05, loc=<-0.123 -15.1423 0>, rot=<-0.0093 -0.707 0.0093 -0.707>}
 clavicle_l {parent=spine_05, loc=<2.5856 -0.2866 -4.3777>, rot=<0.0465 -0.7289 -0.009 -0.6829>}
 upperarm_l {parent=clavicle_l, loc=<18.2264 0.1342 -0.001>, rot=<0.0253 -0.2693 0.0013 -0.9627>}
 lowerarm_l {parent=upperarm_l, loc=<29.0071 0.1686 -1.239>, rot=<0.0391 -0.0099 -0.2453 -0.9686>}
 hand_l {parent=lowerarm_l, loc=<25.8607 0.222 1.0274>, rot=<-0.7193 -0.0988 0.1193 -0.6773>}
 index_metacarpal_l {parent=hand_l, loc=<1.6535 0.0183 -2.96>, rot=<0.0158 -0.0124 -0.0173 -0.9996>}
 index_01_l {parent=index_metacarpal_l, loc=<8.3008 0.0357 -0.0524>, rot=<-0.0044 0.0485 0.0924 -0.9945>}
 index_02_l {parent=index_01_l, loc=<3.3587 -0.3004 0.0452>, rot=<-0.002 0.0027 0.3225 -0.9466>}
 index_03_l {parent=index_02_l, loc=<2.5578 -0.123 -0.0252>, rot=<-0.002 0.0067 0.2558 -0.9667>}
 dfrm_arm_spline_09_l_ctrl_null {parent=hand_l, type=Null, loc=<0.2819 -1.2051 -0.2255>, rot=<-0.9907 -0.1284 0.0454 0.002>}
 dfrm_arm_spline_09_l_ctrl {controlType=EULER_TRANSFORM, parent=dfrm_arm_spline_09_l_ctrl_null, type=Control, loc=<0 0 0>, rot=<0 0 0 1>}
 thumb_01_l {parent=hand_l, loc=<1.3118 -0.3191 -2.459>, rot=<0.6546 -0.4042 0.0664 -0.6354>}
 thumb_02_l {parent=thumb_01_l, loc=<5.9336 -0.0448 0.0401>, rot=<-0.0037 0.0213 0.298 -0.9543>}
 thumb_03_l {parent=thumb_02_l, loc=<2.2223 0.3682 -0.0907>, rot=<0.0118 -0.013 0.0066 -0.9998>}
 ring_metacarpal_l {parent=hand_l, loc=<2.2074 -0.3625 1.3638>, rot=<-0.0284 0.0722 -0.0458 -0.9959>}
 ring_01_l {parent=ring_metacarpal_l, loc=<7.0295 -0.6197 -0.1446>, rot=<-0.0489 -0.0252 0.1892 -0.9804>}
 ring_02_l {parent=ring_01_l, loc=<3.6174 0.4647 -0.1027>, rot=<0 0 0.2806 -0.9598>}
 ring_03_l {parent=ring_02_l, loc=<2.4152 0.114 0.0199>, rot=<0 0 0.1189 -0.9929>}
 middle_metacarpal_l {parent=hand_l, loc=<2.1856 -0.1725 -0.5984>, rot=<0.006 0.0168 -0.0329 -0.9993>}
 middle_01_l {parent=middle_metacarpal_l, loc=<7.3275 -0.2447 0.0129>, rot=<-0.0208 0.0042 0.0858 -0.9961>}
 middle_02_l {parent=middle_01_l, loc=<4.0123 0.1401 0.1175>, rot=<-0.0218 0.0258 0.3512 -0.9357>}
 middle_03_l {parent=middle_02_l, loc=<2.6077 0.0534 0.1373>, rot=<0.0082 0.0188 0.1708 -0.9851>}
 pinky_metacarpal_l {parent=hand_l, loc=<1.7616 -0.9753 3.3814>, rot=<-0.0397 0.0814 -0.069 -0.9935>}
 pinky_01_l {parent=pinky_metacarpal_l, loc=<7.1709 -0.4831 -0.1094>, rot=<-0.0654 -0.0035 0.2242 -0.9723>}
 pinky_02_l {parent=pinky_01_l, loc=<2.8274 0.0893 -0.1043>, rot=<0.0037 0.0011 0.2854 -0.9584>}
 pinky_03_l {parent=pinky_02_l, loc=<2.2046 0.0307 -0.0078>, rot=<-0.0002 0.0098 0.1445 -0.9895>}
 lowerarm_twist_01_l {parent=lowerarm_l, loc=<17.8549 0 0>, rot=<0 0 0 1>}
 lowerarm_twist_02_l {parent=lowerarm_l, loc=<9.6199 0 0>, rot=<0 0 0 1>}
 upperarm_twist_01_l {parent=upperarm_l, loc=<7.9796 0.0843 -0.6195>, rot=<0 0 0 1>}
 upperarm_twist_02_l {parent=upperarm_l, loc=<19.5059 0.0843 -0.6195>, rot=<0 0 0 1>}
 clavicle_r {parent=spine_05, loc=<2.5855 -0.2866 4.3777>, rot=<0.7289 0.0465 -0.6829 0.009>}
 upperarm_r {parent=clavicle_r, loc=<-18.2264 -0.1342 0.001>, rot=<0.0253 -0.2693 0.0013 -0.9627>}
 lowerarm_r {parent=upperarm_r, loc=<-29.0071 -0.1686 1.239>, rot=<0.0391 -0.0099 -0.2453 -0.9686>}
 hand_r {parent=lowerarm_r, loc=<-25.8607 -0.222 -1.0274>, rot=<-0.7193 -0.0988 0.1193 -0.6773>}
 index_metacarpal_r {parent=hand_r, loc=<-1.6535 -0.0183 2.96>, rot=<0.0158 -0.0124 -0.0173 -0.9996>}
 index_01_r {parent=index_metacarpal_r, loc=<-8.3008 -0.0357 0.0524>, rot=<-0.0044 0.0485 0.0924 -0.9945>}
 index_02_r {parent=index_01_r, loc=<-3.3587 0.3004 -0.0452>, rot=<-0.002 0.0027 0.3225 -0.9466>}
 index_03_r {parent=index_02_r, loc=<-2.5578 0.123 0.0252>, rot=<-0.002 0.0067 0.2558 -0.9667>}
 dfrm_arm_spline_09_r_ctrl_null {parent=hand_r, type=Null, loc=<-0.2818 1.2051 0.2255>, rot=<-0.9907 -0.1284 0.0454 0.002>}
 dfrm_arm_spline_09_r_ctrl {controlType=EULER_TRANSFORM, parent=dfrm_arm_spline_09_r_ctrl_null, type=Control, loc=<0 0 0>, rot=<0 0 0 1>}
 thumb_01_r {parent=hand_r, loc=<-1.3118 0.3191 2.459>, rot=<0.6546 -0.4042 0.0664 -0.6354>}
 thumb_02_r {parent=thumb_01_r, loc=<-5.9336 0.0448 -0.0401>, rot=<-0.0037 0.0213 0.298 -0.9543>}
 thumb_03_r {parent=thumb_02_r, loc=<-2.2223 -0.3682 0.0907>, rot=<0.0118 -0.013 0.0066 -0.9998>}
 ring_metacarpal_r {parent=hand_r, loc=<-2.2074 0.3625 -1.3638>, rot=<-0.0284 0.0722 -0.0458 -0.9959>}
 ring_01_r {parent=ring_metacarpal_r, loc=<-7.0295 0.6197 0.1446>, rot=<-0.0489 -0.0252 0.1892 -0.9804>}
 ring_02_r {parent=ring_01_r, loc=<-3.6174 -0.4647 0.1027>, rot=<0 0 0.2806 -0.9598>}
 ring_03_r {parent=ring_02_r, loc=<-2.4152 -0.114 -0.0199>, rot=<0 0 0.1189 -0.9929>}
 middle_metacarpal_r {parent=hand_r, loc=<-2.1856 0.1725 0.5984>, rot=<0.006 0.0168 -0.0329 -0.9993>}
 middle_01_r {parent=middle_metacarpal_r, loc=<-7.3275 0.2447 -0.0129>, rot=<-0.0208 0.0042 0.0858 -0.9961>}
 middle_02_r {parent=middle_01_r, loc=<-4.0123 -0.1401 -0.1175>, rot=<-0.0218 0.0258 0.3512 -0.9357>}
 middle_03_r {parent=middle_02_r, loc=<-2.6077 -0.0533 -0.1373>, rot=<0.0082 0.0188 0.1708 -0.9851>}
 pinky_metacarpal_r {parent=hand_r, loc=<-1.7616 0.9753 -3.3814>, rot=<-0.0397 0.0814 -0.069 -0.9935>}
 pinky_01_r {parent=pinky_metacarpal_r, loc=<-7.1709 0.4831 0.1094>, rot=<-0.0654 -0.0035 0.2242 -0.9723>}
 pinky_02_r {parent=pinky_01_r, loc=<-2.8274 -0.0892 0.1043>, rot=<0.0037 0.0011 0.2854 -0.9584>}
 pinky_03_r {parent=pinky_02_r, loc=<-2.2046 -0.0307 0.0078>, rot=<-0.0002 0.0098 0.1445 -0.9895>}
 lowerarm_twist_01_r {parent=lowerarm_r, loc=<-17.855 0 0>, rot=<0 0 0 1>}
 lowerarm_twist_02_r {parent=lowerarm_r, loc=<-9.6199 0 0>, rot=<0 0 0 1>}
 upperarm_twist_01_r {parent=upperarm_r, loc=<-7.9796 -0.0843 0.6195>, rot=<0 0 0 1>}
 upperarm_twist_02_r {parent=upperarm_r, loc=<-19.506 -0.0843 0.6195>, rot=<0 0 0 1>}
 dfrm_arm_spline_mstr_l {parent=spine_05, loc=<-0.5674 -2.1692 -22.4246>, rot=<0.3738 -0.625 -0.6214 -0.2889>}
 dfrm_clavicle_l {parent=dfrm_arm_spline_mstr_l, loc=<0 0 0>, rot=<0 0 0 1>}
 dfrm_arm_spline_02_l {parent=dfrm_clavicle_l, loc=<6.6972 0 0>, rot=<0 0 0 1>}
 dfrm_arm_spline_03_l {parent=dfrm_arm_spline_02_l, loc=<6.6969 0 0>, rot=<0 0.0321 0 -0.9995>}
 dfrm_arm_spline_04_l {parent=dfrm_arm_spline_03_l, loc=<6.6977 0 0>, rot=<0.0005 0.0364 -0.0134 -0.9992>}
 dfrm_lowerarm_l {parent=dfrm_arm_spline_04_l, loc=<6.6973 0 0>, rot=<0 0.095 0 -0.9955>}
 dfrm_arm_spline_06_l {parent=dfrm_lowerarm_l, loc=<6.697 0 0>, rot=<0.0008 0.1009 -0.0079 -0.9949>}
 dfrm_arm_spline_07_l {parent=dfrm_arm_spline_06_l, loc=<6.6971 0 0>, rot=<0.0134 0.0103 -0.0218 -0.9996>}
 dfrm_arm_spline_08_l {parent=dfrm_arm_spline_07_l, loc=<6.6973 0 0>, rot=<0 0 -0.0187 -0.9998>}
 dfrm_arm_spline_09_l {parent=dfrm_arm_spline_08_l, loc=<6.6976 0 0>, rot=<0 0 0 1>}
 dfrm_arm_spline_mstr_r {parent=spine_05, loc=<-0.567 -2.1692 22.4246>, rot=<-0.625 -0.3738 0.2889 -0.6214>}
 dfrm_clavicle_r {parent=dfrm_arm_spline_mstr_r, loc=<0 0 0>, rot=<0 0 0 1>}
 dfrm_arm_spline_02_r {parent=dfrm_clavicle_r, loc=<-6.6976 -0.0005 0>, rot=<0 0 0 -1>}
 dfrm_arm_spline_03_r {parent=dfrm_arm_spline_02_r, loc=<-6.6968 0.0002 0>, rot=<0 0.0321 0 -0.9995>}
 dfrm_arm_spline_04_r {parent=dfrm_arm_spline_03_r, loc=<-6.6977 0.0001 0>, rot=<0.0005 0.0364 -0.0134 -0.9992>}
 dfrm_lowerarm_r {parent=dfrm_arm_spline_04_r, loc=<-6.6976 -0.0005 0>, rot=<0 0.095 0 -0.9955>}
 dfrm_arm_spline_06_r {parent=dfrm_lowerarm_r, loc=<-6.6966 0.0005 -0.0001>, rot=<0.0008 0.1009 -0.0079 -0.9949>}
 dfrm_arm_spline_07_r {parent=dfrm_arm_spline_06_r, loc=<-6.6975 -0.0005 0.0002>, rot=<0.0134 0.0103 -0.0218 -0.9996>}
 dfrm_arm_spline_08_r {parent=dfrm_arm_spline_07_r, loc=<-6.6971 0.0003 -0.0001>, rot=<0 0 -0.0187 -0.9998>}
 dfrm_arm_spline_09_r {parent=dfrm_arm_spline_08_r, loc=<-6.6975 0.0001 0>, rot=<0 0 0 -1>}
 dfrm_arm_spline_01_l_ctrl_null {parent=spine_05, type=Null, loc=<-0.5675 -2.1692 -22.4246>, rot=<0.3738 -0.625 -0.6214 -0.2889>}
 dfrm_arm_spline_01_l_ctrl {controlType=EULER_TRANSFORM, parent=dfrm_arm_spline_01_l_ctrl_null, type=Control, loc=<0 0 0>, rot=<0 0 0 1>}
 dfrm_arm_spline_01_r_ctrl_null {parent=spine_05, type=Null, loc=<-0.567 -2.1692 22.4246>, rot=<0.625 0.3738 -0.2889 0.6214>}
 dfrm_arm_spline_01_r_ctrl {controlType=EULER_TRANSFORM, parent=dfrm_arm_spline_01_r_ctrl_null, type=Control, loc=<0 0 0>, rot=<0 0 0 1>}
 thigh_l {parent=pelvis, loc=<-1.8716 -0.0109 -11.9936>, rot=<0.0012 0.0572 0.0155 -0.9982>}
 calf_l {parent=thigh_l, loc=<-41.9178 -0.2246 0.0926>, rot=<0.0006 -0.0161 -0.0391 -0.9991>}
 foot_l {parent=calf_l, loc=<-38.7057 1.2438 0.2452>, rot=<0.0975 -0.0474 0.0317 -0.9936>}
 dfrm_leg_spline_09_l_ctrl_null {parent=foot_l, type=Null, loc=<0.7311 -0.3892 0.329>, rot=<0.0316 0.1005 -0.9944 0.0057>}
 dfrm_leg_spline_09_l_ctrl {controlType=EULER_TRANSFORM, parent=dfrm_leg_spline_09_l_ctrl_null, type=Control, loc=<0 0 0>, rot=<0 0 0 1>}
 ball_l {parent=foot_l, loc=<-6.1614 13.1469 0.6807>, rot=<0.0001 -0.0001 -0.7063 -0.7079>}
 calf_twist_01_l {parent=calf_l, loc=<-32.9335 -0.0967 0.0397>, rot=<0.0237 0.0095 -0.0035 -0.9997>}
 calf_twist_02_l {parent=calf_l, loc=<-15.9268 -0.0967 0.0397>, rot=<0.0237 0.0095 -0.0035 -0.9997>}
 thigh_twist_01_l {parent=thigh_l, loc=<-21.6303 0 0>, rot=<0.0013 0 0 -1>}
 thigh_r {parent=pelvis, loc=<-1.8716 -0.0109 11.9936>, rot=<0.0572 -0.0012 0.9982 0.0155>}
 calf_r {parent=thigh_r, loc=<41.9178 0.2246 -0.0926>, rot=<0.0006 -0.0161 -0.0391 -0.9991>}
 foot_r {parent=calf_r, loc=<38.7057 -1.2438 -0.2452>, rot=<0.0975 -0.0474 0.0317 -0.9936>}
 dfrm_leg_spline_09_r_ctrl_null {parent=foot_r, type=Null, loc=<0 0.3892 -0.3291>, rot=<0.0316 0.1005 -0.9944 0.0057>}
 dfrm_leg_spline_09_r_ctrl {controlType=EULER_TRANSFORM, parent=dfrm_leg_spline_09_r_ctrl_null, type=Control, loc=<0 0 0>, rot=<0 0 0 1>}
 ball_r {parent=foot_r, loc=<6.1614 -13.1469 -0.6807>, rot=<0.0001 -0.0001 -0.7063 -0.7079>}
 calf_twist_01_r {parent=calf_r, loc=<32.9335 0.0967 -0.0397>, rot=<0.0237 0.0095 -0.0035 -0.9997>}
 calf_twist_02_r {parent=calf_r, loc=<15.9268 0.0967 -0.0397>, rot=<0.0237 0.0095 -0.0035 -0.9997>}
 thigh_twist_01_r {parent=thigh_r, loc=<21.6303 0 0>, rot=<0.0013 0 0 -1>}
 dfrm_leg_spline_01_l {parent=pelvis, loc=<-2.8524 -0.4318 -14.7872>, rot=<-0.0267 0.0009 0.9996 0.0031>}
 dfrm_leg_spline_02_l {parent=dfrm_leg_spline_01_l, loc=<9.8342 -0.0001 -0.0005>, rot=<0 -0.0005 0.0001 -1>}
 dfrm_leg_spline_03_l {parent=dfrm_leg_spline_02_l, loc=<9.8342 -0.0002 -0.0007>, rot=<0 -0.0008 0.0003 -1>}
 dfrm_leg_spline_04_l {parent=dfrm_leg_spline_03_l, loc=<9.8342 -0.0002 -0.0012>, rot=<0 -0.0013 0.0002 -1>}
 dfrm_leg_spline_05_l {parent=dfrm_leg_spline_04_l, loc=<9.8342 -0.0001 0>, rot=<0 0 0.0001 -1>}
 dfrm_leg_spline_06_l {parent=dfrm_leg_spline_05_l, loc=<9.8342 -0.0002 0.0006>, rot=<0 0.0007 0.0002 -1>}
 dfrm_leg_spline_07_l {parent=dfrm_leg_spline_06_l, loc=<9.8342 0 0.0003>, rot=<0 0.0003 0.0001 -1>}
 dfrm_leg_spline_08_l {parent=dfrm_leg_spline_07_l, loc=<9.8342 -0.0001 0.0003>, rot=<0 0.0003 0.0001 -1>}
 dfrm_leg_spline_09_l {parent=dfrm_leg_spline_08_l, loc=<9.8342 0 0>, rot=<0 0 0 1>}
 dfrm_leg_spline_01_r {parent=pelvis, loc=<-2.8523 -0.4318 14.7872>, rot=<-0.0009 -0.0267 0.0031 -0.9996>}
 dfrm_leg_spline_02_r {parent=dfrm_leg_spline_01_r, loc=<-9.8342 0.0001 0.0004>, rot=<0 -0.0005 0.0001 -1>}
 dfrm_leg_spline_03_r {parent=dfrm_leg_spline_02_r, loc=<-9.8343 0.0002 0.0008>, rot=<0 -0.0008 0.0003 -1>}
 dfrm_leg_spline_04_r {parent=dfrm_leg_spline_03_r, loc=<-9.8342 0.0002 0.0012>, rot=<0 -0.0013 0.0002 -1>}
 dfrm_leg_spline_05_r {parent=dfrm_leg_spline_04_r, loc=<-9.8342 0.0001 0>, rot=<0 0 0.0001 -1>}
 dfrm_leg_spline_06_r {parent=dfrm_leg_spline_05_r, loc=<-9.8342 0.0002 -0.0006>, rot=<0 0.0007 0.0002 -1>}
 dfrm_leg_spline_07_r {parent=dfrm_leg_spline_06_r, loc=<-9.8342 0 -0.0003>, rot=<0 0.0003 0.0001 -1>}
 dfrm_leg_spline_08_r {parent=dfrm_leg_spline_07_r, loc=<-9.8342 0 -0.0003>, rot=<0 0.0003 0.0001 -1>}
 dfrm_leg_spline_09_r {parent=dfrm_leg_spline_08_r, loc=<-9.8342 0 0>, rot=<0 0 0 1>}
 dfrm_leg_spline_01_l_ctrl_null {parent=pelvis, type=Null, loc=<-2.8524 -0.4318 -14.7872>, rot=<-0.0267 0.0009 0.9996 0.0031>}
 dfrm_leg_spline_01_l_ctrl {controlType=EULER_TRANSFORM, parent=dfrm_leg_spline_01_l_ctrl_null, type=Control, loc=<0 0 0>, rot=<0 0 0 1>}
 dfrm_leg_spline_03_l_ctrl_null {parent=pelvis, type=Null, loc=<-22.4917 -0.309 -15.849>, rot=<-0.028 0.0009 0.9996 0.0035>}
 dfrm_leg_spline_03_l_ctrl {controlType=EULER_TRANSFORM, parent=dfrm_leg_spline_03_l_ctrl_null, type=Control, loc=<0 0 0>, rot=<0 0 0 1>}
 dfrm_leg_spline_05_l_ctrl_null {parent=pelvis, type=Null, loc=<-42.1272 -0.1705 -16.9771>, rot=<-0.0293 0.0009 0.9996 0.0038>}
 dfrm_leg_spline_05_l_ctrl {controlType=EULER_TRANSFORM, parent=dfrm_leg_spline_05_l_ctrl_null, type=Control, loc=<0 0 0>, rot=<0 0 0 1>}
 dfrm_leg_spline_07_l_ctrl_null {parent=pelvis, type=Null, loc=<-61.7621 -0.0207 -18.1156>, rot=<-0.0283 0.0009 0.9996 0.004>}
 dfrm_leg_spline_07_l_ctrl {controlType=EULER_TRANSFORM, parent=dfrm_leg_spline_07_l_ctrl_null, type=Control, loc=<0 0 0>, rot=<0 0 0 1>}
 dfrm_arm_spline_03_l_ctrl_null {parent=pelvis, type=Null, loc=<30.2636 -6.8928 -33.4849>, rot=<-0.3174 0.6348 0.6141 0.3451>}
 dfrm_arm_spline_03_l_ctrl {controlType=EULER_TRANSFORM, parent=dfrm_arm_spline_03_l_ctrl_null, type=Control, loc=<0 0 0>, rot=<0 0 0 1>}
 dfrm_arm_spline_05_l_ctrl_null {parent=pelvis, type=Null, loc=<22.643 -6.1192 -44.4613>, rot=<-0.2253 0.5888 0.6545 0.4173>}
 dfrm_arm_spline_05_l_ctrl {controlType=EULER_TRANSFORM, parent=dfrm_arm_spline_05_l_ctrl_null, type=Control, loc=<0 0 0>, rot=<0 0 0 1>}
 dfrm_arm_spline_07_l_ctrl_null {parent=pelvis, type=Null, loc=<15.5271 -1.1005 -54.5482>, rot=<-0.1412 0.5342 0.6965 0.4578>}
 dfrm_arm_spline_07_l_ctrl {controlType=EULER_TRANSFORM, parent=dfrm_arm_spline_07_l_ctrl_null, type=Control, loc=<0 0 0>, rot=<0 0 0 1>}
 dfrm_arm_spline_03_r_ctrl_null {parent=pelvis, type=Null, loc=<30.2637 -6.8928 33.4849>, rot=<0.6348 0.3174 -0.3451 0.6141>}
 dfrm_arm_spline_03_r_ctrl {controlType=EULER_TRANSFORM, parent=dfrm_arm_spline_03_r_ctrl_null, type=Control, loc=<0 0 0>, rot=<0 0 0 1>}
 dfrm_arm_spline_05_r_ctrl_null {parent=pelvis, type=Null, loc=<22.6425 -6.1192 44.4614>, rot=<0.5888 0.2253 -0.4173 0.6545>}
 dfrm_arm_spline_05_r_ctrl {controlType=EULER_TRANSFORM, parent=dfrm_arm_spline_05_r_ctrl_null, type=Control, loc=<0 0 0>, rot=<0 0 0 1>}
 dfrm_arm_spline_07_r_ctrl_null {parent=pelvis, type=Null, loc=<15.5268 -1.1005 54.5482>, rot=<0.5342 0.1412 -0.4578 0.6965>}
 dfrm_arm_spline_07_r_ctrl {controlType=EULER_TRANSFORM, parent=dfrm_arm_spline_07_r_ctrl_null, type=Control, loc=<0 0 0>, rot=<0 0 0 1>}
 dfrm_leg_spline_01_r_null {parent=pelvis, type=Null, loc=<-2.8523 -0.4318 14.7872>, rot=<-0.0009 -0.0267 0.0031 -0.9996>}
 dfrm_leg_spline_01_r_ctrl_null {parent=pelvis, type=Null, loc=<-2.8523 -0.4318 14.7872>, rot=<0.0009 0.0267 -0.0031 0.9996>}
 dfrm_leg_spline_01_r_ctrl {controlType=EULER_TRANSFORM, parent=dfrm_leg_spline_01_r_ctrl_null, type=Control, loc=<0 0 0>, rot=<0 0 0 1>}
 dfrm_leg_spline_03_r_ctrl_null {parent=pelvis, type=Null, loc=<-22.4917 -0.309 15.849>, rot=<0.0009 0.028 -0.0035 0.9996>}
 dfrm_leg_spline_03_r_ctrl {controlType=EULER_TRANSFORM, parent=dfrm_leg_spline_03_r_ctrl_null, type=Control, loc=<0 0 0>, rot=<0 0 0 1>}
 dfrm_leg_spline_05_r_ctrl_null {parent=pelvis, type=Null, loc=<-42.1273 -0.1705 16.9771>, rot=<0.0009 0.0293 -0.0038 0.9996>}
 dfrm_leg_spline_05_r_ctrl {controlType=EULER_TRANSFORM, parent=dfrm_leg_spline_05_r_ctrl_null, type=Control, loc=<0 0 0>, rot=<0 0 0 1>}
 dfrm_leg_spline_07_r_ctrl_null {parent=pelvis, type=Null, loc=<-61.7621 -0.0207 18.1156>, rot=<0.0009 0.0283 -0.004 0.9996>}
 dfrm_leg_spline_07_r_ctrl {controlType=EULER_TRANSFORM, parent=dfrm_leg_spline_07_r_ctrl_null, type=Control, loc=<0 0 0>, rot=<0 0 0 1>}
 def_groin_r_base {parent=pelvis, loc=<-7.1198 3.3828 15.1821>, rot=<0.042 -0.0009 0.999 0.0147>}
 def_groin_r {parent=def_groin_r_base, loc=<0 0 0>, rot=<0 0 0 1>}
 def_groin_l_base {parent=pelvis, loc=<-7.1198 3.3828 -15.1821>, rot=<0.0009 0.042 0.0147 -0.999>}
 def_groin_l {parent=def_groin_l_base, loc=<0 0 0>, rot=<0 0 0 1>}
 def_hip_r_base {parent=pelvis, loc=<-7.5892 -5.0111 15.1963>, rot=<-0.0147 -0.999 -0.0009 0.042>}
 def_hip_r {parent=def_hip_r_base, loc=<0 0 0>, rot=<0 0 0 1>}
 def_hip_l_base {parent=pelvis, loc=<-7.5892 -5.0111 -15.1963>, rot=<0.999 -0.0147 0.042 0.0009>}
 def_hip_l {parent=def_hip_l_base, loc=<0 0 0>, rot=<0 0 0 -1>}
 dfrm_rowerarm_r_null {type=Null, loc=<-44.4614 -6.0362 115.269>, rot=<-0.7002 0.3013 -0.1267 -0.6348>}
]
variables [
 UseBackpack_PuzzleShed {type=double, default=0}
 arm_l_bind_pose {type=FRigPose, default=(Elements=,HierarchyTopologyVersion=-1,PoseHash=-1)}
 arm_l_ctrls {
  type="TArray<FRigElementKey>"
  default=((Type=Control,Name="dfrm_arm_spline_01_l_ctrl"),(Type=Control,Name="dfrm_arm_spline_03_l_ctrl"),(Type=Control,Name="dfrm_arm_spline_05_l_ctrl"),(Type=Control,Name="dfrm_arm_spline_07_l_ctrl"),(Type=Control,Name="dfrm_arm_spline_09_l_ctrl"))
 }
 arm_r_bind_pose {type=FRigPose, default=(Elements=,HierarchyTopologyVersion=-1,PoseHash=-1)}
 arm_r_ctrls {
  type="TArray<FRigElementKey>"
  default=((Type=Control,Name="dfrm_arm_spline_01_r_ctrl"),(Type=Control,Name="dfrm_arm_spline_03_r_ctrl"),(Type=Control,Name="dfrm_arm_spline_05_r_ctrl"),(Type=Control,Name="dfrm_arm_spline_07_r_ctrl"),(Type=Control,Name="dfrm_arm_spline_09_r_ctrl"))
 }
 arms {type=double, default=0}
 arms_0 {type=double, default=0}
 arms_1 {type=double, default=0}
 calf_l_bind_pose {type=FRigPose, default=(Elements=,HierarchyTopologyVersion=-1,PoseHash=-1)}
 calf_l_ctrls {
  type="TArray<FRigElementKey>"
  default=((Type=Control,Name="dfrm_leg_spline_01_l_ctrl"),(Type=Control,Name="dfrm_leg_spline_03_l_ctrl"),(Type=Control,Name="dfrm_leg_spline_05_l_ctrl"),(Type=Control,Name="dfrm_leg_spline_07_l_ctrl"),(Type=Control,Name="dfrm_leg_spline_09_l_ctrl"))
 }
 calf_r_bind_pose {type=FRigPose, default=(Elements=,HierarchyTopologyVersion=-1,PoseHash=-1)}
 calf_r_ctrls {
  type="TArray<FRigElementKey>"
  default=((Type=Control,Name="dfrm_leg_spline_01_l_ctrl"),(Type=Control,Name="dfrm_leg_spline_03_l_ctrl"),(Type=Control,Name="dfrm_leg_spline_05_l_ctrl"),(Type=Control,Name="dfrm_leg_spline_07_l_ctrl"),(Type=Control,Name="dfrm_leg_spline_09_l_ctrl"))
 }
 chest_twist_alpha {type=double, default=0}
 knee_l_rot_pose {type=FRigPose, default=(Elements=,HierarchyTopologyVersion=-1,PoseHash=-1)}
 knee_r_rot_pose {type=FRigPose, default=(Elements=,HierarchyTopologyVersion=-1,PoseHash=-1)}
 leg_l_bind_pose {type=FRigPose, default=(Elements=,HierarchyTopologyVersion=-1,PoseHash=-1)}
 leg_l_ctrls {
  type="TArray<FRigElementKey>"
  default=((Type=Control,Name="dfrm_leg_spline_01_l_ctrl"),(Type=Control,Name="dfrm_leg_spline_03_l_ctrl"),(Type=Control,Name="dfrm_leg_spline_05_l_ctrl"),(Type=Control,Name="dfrm_leg_spline_07_l_ctrl"),(Type=Control,Name="dfrm_leg_spline_09_l_ctrl"))
 }
 leg_r_bind_pose {type=FRigPose, default=(Elements=,HierarchyTopologyVersion=-1,PoseHash=-1)}
 leg_r_ctrls {
  type="TArray<FRigElementKey>"
  default=((Type=Control,Name="dfrm_leg_spline_01_r_ctrl"),(Type=Control,Name="dfrm_leg_spline_03_r_ctrl"),(Type=Control,Name="dfrm_leg_spline_05_r_ctrl"),(Type=Control,Name="dfrm_leg_spline_07_r_ctrl"),(Type=Control,Name="dfrm_leg_spline_09_r_ctrl"))
 }
 spr_knee_l_rot_alpha {type=double, default=0}
 spr_knee_r_rot_alpha {type=double, default=0}
 spr_thigh_l_up_alpha {type=double, default=0}
 spr_thigh_r_up_alpha {type=double, default=0}
 spr_upper_l_DBNO {type=double, default=0}
 spr_upper_l_front_alpha {type=double, default=0}
 spr_upper_l_up_alpha {type=double, default=0}
 spr_upper_r_front_alpha {type=double, default=0}
 spr_upper_r_up_alpha {type=double, default=0}
 thigh_l_bind_pose {type=FRigPose, default=(Elements=,HierarchyTopologyVersion=-1,PoseHash=-1)}
 thigh_l_ctrls {
  type="TArray<FRigElementKey>"
  default=((Type=Control,Name="dfrm_leg_spline_01_l_ctrl"),(Type=Control,Name="dfrm_leg_spline_03_l_ctrl"),(Type=Control,Name="dfrm_leg_spline_05_l_ctrl"),(Type=Control,Name="dfrm_leg_spline_07_l_ctrl"),(Type=Control,Name="dfrm_leg_spline_09_l_ctrl"))
 }
 thigh_l_up_pose {type=FRigPose, default=(Elements=,HierarchyTopologyVersion=-1,PoseHash=-1)}
 thigh_r_bind_pose {type=FRigPose, default=(Elements=,HierarchyTopologyVersion=-1,PoseHash=-1)}
 thigh_r_ctrls {
  type="TArray<FRigElementKey>"
  default=((Type=Control,Name="dfrm_leg_spline_01_l_ctrl"),(Type=Control,Name="dfrm_leg_spline_03_l_ctrl"),(Type=Control,Name="dfrm_leg_spline_05_l_ctrl"),(Type=Control,Name="dfrm_leg_spline_07_l_ctrl"),(Type=Control,Name="dfrm_leg_spline_09_l_ctrl"))
 }
 thigh_r_up_pose {type=FRigPose, default=(Elements=,HierarchyTopologyVersion=-1,PoseHash=-1)}
 upperarm_l_front_pose {type=FRigPose, default=(Elements=,HierarchyTopologyVersion=-1,PoseHash=-1)}
 upperarm_r_front_pose {type=FRigPose, default=(Elements=,HierarchyTopologyVersion=-1,PoseHash=-1)}
]
